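# Editor utility list-entry widget in the Sound Library editor that edits one sound and the gameplay tags mapped to it.
# Two SinglePropertyViews expose its own Sound and SoundLibraryTags variables, filled from the assigned EditEntry object.
# Sound changes update linked library entries and the library's tag-to-sound map.
# Tag changes add or remove map keys, rebuild the parent's tag tree and restore selection after a short delay, with edits wrapped in undo transactions.
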WidgetBlueprint /Game/Sounds/Fort_Audio_Blueprints/Utilities/EditorUtilityWidgets/SoundLibEditor/EUW_SoundLibraryEditor_EditEntry
graphs [
 EventGraph {nodes [
   "On Property Changed (Sounds)"!() -> (OutputDelegate, then, PropertyName:Name=None) {is=title K2Node_ComponentBoundEvent}
   n1 = Get EntryObjectCasted() -> (EntryObjectCasted:Sound Library Editor Edit Entry Object Reference)
   > Set Sound!(
    execute
    Sound:Sound Base Soft Object Reference=Get Sound() -> (Sound:Sound Base Soft Object Reference)
    self:Sound Library Editor Edit Entry Object Reference=@n1
   ) -> (then, Output_Get:Sound Base Soft Object Reference)
   > TransactionStart!(execute, self, Description:Text="To Text (String)"(InString=OnSoundChanged)) -> (then)
   > n2 = Values!(
    execute
    TargetMap:Map of Sound Library Tag Structures to Sound Library Editor Library Entry Object References=Get TagToLibraryEntry!(self:Sound Library Editor Edit Entry Object Reference=@n1) -> (TagToLibraryEntry:Map of Sound Library Tag Structures to Sound Library Editor Library Entry Object References)
   ) -> (then, Values:Array of Sound Library Editor Library Entry Object References)
   > n3 = For Each Loop(Array:Array of Sound Library Editor Library Entry Object References=@n2) -> (Array Element:Sound Library Editor Library Entry Object Reference)
   TransactionEnd(execute=@n3.Completed)
   Set Sound(execute=@n3.LoopBody, Sound=Get Sound(), self:Sound Library Editor Library Entry Object Reference=@n3)
   n4 = Get Widget!(self:Sound Library Editor Library Entry Object Reference=@n3) -> (Widget:EUW Sound Library Editor Library Entry Object Reference)
   > Refresh Display(self:EUW Sound Library Editor Library Entry Object Reference=@n4)
   n5 = Get Library!(self:EUW Sound Library Editor Object Reference=Get Parent!(self:EUW Sound Library Editor Library Entry Object Reference=@n4) -> (Parent:EUW Sound Library Editor Object Reference)) -> (Library:Sound Library Object Reference)
   > TransactionAdd!(execute, self, Object:Object Reference=@n5) -> (then)
   > Add!(
    execute
    TargetMap:Map of Gameplay Tag Structures to Sound Base Soft Object References=Get Sounds!(self:Sound Library Object Reference=@n5) -> (Sounds:Map of Gameplay Tag Structures to Sound Base Soft Object References)
    Key:"Gameplay Tag Structure (by ref)"=Get Tag!(self:Sound Library Editor Library Entry Object Reference=@n3) -> (Tag:Sound Library Tag Structure)
    Value:"Sound Base Soft Object Reference (by ref)"=Get Sound()
   ) -> (then) {is=title K2Node_CallFunction}
   "On Property Changed (Tags)"!() -> (OutputDelegate, then, PropertyName:Name=None) {is=title K2Node_ComponentBoundEvent}
   > Delay!(execute, WorldContextObject:Object Reference=Self-Reference(), Duration=0.05) -> (then)
   > OnTagPropertyChanged()
   n6 = RefreshSelection!() -> (OutputDelegate, then, Parent:EUW Sound Library Editor Object Reference) {is=title K2Node_CustomEvent}
   > Delay(WorldContextObject=Self-Reference(), Duration=0.05)
   > OnEntryClicked!(
    execute
    self:EUW Sound Library Editor Object Reference=@n6.Parent
    Widget:EUW Sound Library Editor Library Entry Object Reference
    DoubleClick:bool=false
    ForceAdditiveSelect:bool=false
   ) -> (then)
   > n7 = For Each Loop(Array:Array of Gameplay Tag Structures=Get SelectionTags() -> (SelectionTags:Array of Gameplay Tag Structures)) -> (Array Element:Gameplay Tag Structure)
   Clear(execute=@n7.Completed, TargetArray:Array of Gameplay Tag Structures=Get SelectionTags())
   n8 = FindEntryByTag!(
    self:EUW Sound Library Editor Object Reference=@n6.Parent
    TargetArray:Array of Sound Library Editor Library Entry Object References=Get AllEntryObjects!(self:EUW Sound Library Editor Object Reference=@n6.Parent) -> (AllEntryObjects:Array of Sound Library Editor Library Entry Object References)
    Tag:"Gameplay Tag Structure (by ref)"=@n7
   ) -> (Entry:Sound Library Editor Library Entry Object Reference)
   Is Valid(exec=@n7.LoopBody, InputObject=@n8)
   > OnEntryClicked(self=@n6.Parent, Widget=Get Widget(self=@n8), DoubleClick=false, ForceAdditiveSelect=true)
   n9 = Event OnListItemObjectSet()
   > n10 = Set EntryObject(EntryObject:Object Reference=@n9.ListItemObject) -> (Output_Get:Object Reference)
   > n11 = Cast To SoundLibraryEditor_EditEntry!(execute, Object=@n10) -> (then, CastFailed, AsSound Library Editor Edit Entry:Sound Library Editor Edit Entry Object Reference) {is=title K2Node_DynamicCast}
   > n12 = Set EntryObjectCasted(EntryObjectCasted:Sound Library Editor Edit Entry Object Reference=@n11) -> (Output_Get:Sound Library Editor Edit Entry Object Reference)
   > Set Sound!(execute, Sound=Get Sound!(self:Sound Library Editor Edit Entry Object Reference=@n12) -> (Sound)) -> (then, Output_Get)
   > Set SoundLibraryTags(SoundLibraryTags:Gameplay Tag Container Structure=Get Tag(self:Sound Library Editor Edit Entry Object Reference=@n12) -> (Tag:Gameplay Tag Container Structure)) -> (Output_Get:Gameplay Tag Container Structure)
   Event PreConstruct()
   n13 = Self-Reference()
   > SetObject(self=Get Sounds!() -> (Sounds:Single Property View Object Reference), NewObject=@n13)
   > SetObject(self=Get Tags() -> (Tags:Single Property View Object Reference), NewObject=@n13)
  ]}
 OnTagPropertyChanged {nodes [
   OnTagPropertyChanged() {is=title K2Node_FunctionEntry}
   > Set SoundLibrary(SoundLibrary:Sound Library Object Reference=Get Library(self:Sound Library Editor Edit Entry Object Reference=Get EntryObjectCasted())) -> (Output_Get:Sound Library Object Reference)
   n1 = Get ActionType() -> (ActionType:SoundLibraryEditor_ActionTypes Enum=NewEnumerator0)
   > TransactionStart(Description="To Text (String)"(InString=Enum to String!(Enumerator:SoundLibraryEditor_ActionTypes Enum=@n1) -> (ReturnValue:String)))
   > TransactionAdd(Object=Get SoundLibrary() -> (SoundLibrary:Sound Library Object Reference))
   n2 = Get EntryObjectCasted()
   > Set SelectedEntries(SelectedEntries:Array of Object References=Get SelectedEntries!(self:EUW Sound Library Editor Object Reference=Get ParentWidget!(self:Sound Library Editor Edit Entry Object Reference=@n2) -> (ParentWidget:EUW Sound Library Editor Object Reference)) -> (SelectedEntries:Array of Object References)) -> (Output_Get:Array of Object References)
   > Set TagTracking(TagTracking:Array of Gameplay Tag Structures=Get!() -> (SoundLibraryTags_GameplayTags:Array of Gameplay Tag Structures)) -> (Output_Get:Array of Gameplay Tag Structures)
   > n3 = Keys!(
    execute
    TargetMap:Map of Sound Library Tag Structures to Sound Library Editor Library Entry Object References=Get TagToLibraryEntry(self=Get EntryObjectCasted())
   ) -> (then, Keys:Array of Sound Library Tag Structures)
   > n4 = For Each Loop(Array:Array of Sound Library Tag Structures=@n3) -> (Array Element:Sound Library Tag Structure)
   n5 = For Each Loop(Exec=@n4.Completed, Array:Array of Gameplay Tag Structures=Get TagTracking() -> (TagTracking:Array of Gameplay Tag Structures)) -> (Array Element:Gameplay Tag Structure)
   n6 = Get ParentWidget(self=Get EntryObjectCasted())
   TransactionEnd(execute=@n5.Completed, self:EUW Sound Library Editor Object Reference=@n6)
   > n7 = For Each Loop(Array:Array of Object References=Get SelectedEntries!() -> (SelectedEntries)) -> (Array Element:Object Reference)
   GenerateTagTree(execute=@n7.Completed, self:EUW Sound Library Editor Object Reference=@n6)
   > RefreshSelection!(execute, self, Parent:EUW Sound Library Editor Object Reference=@n6) -> (then) {is=title K2Node_CallFunction}
   n8 = Cast To SoundLibraryEditor_LibraryEntry!(execute=@n7.LoopBody, Object=@n7) -> (then, CastFailed, AsSound Library Editor Library Entry:Sound Library Editor Library Entry Object Reference) {is=title K2Node_DynamicCast}
   > Add Unique(
    TargetArray:Array of Gameplay Tag Structures=Get SelectionTags()
    NewItem:"Gameplay Tag Structure (by ref)"=Get Tag(self:Sound Library Editor Library Entry Object Reference=@n8) -> (Tag:Sound Library Tag Structure)
   )
   Add(
    execute=@n5.LoopBody
    TargetMap=Get Sounds!(self:Sound Library Object Reference=Get SoundLibrary()) -> (Sounds:Map of Gameplay Tag Structures to Sound Base Soft Object References)
    Key=@n5
    Value=Get Sound!() -> (Sound)
   )
   > Add Unique(TargetArray=Get SelectionTags(), NewItem=@n5)
   n9 = Branch(
    execute=@n4.LoopBody
    Condition=HasTag!(
     TagContainer:"Gameplay Tag Container Structure (by ref)"=Get SoundLibraryTags() -> (SoundLibraryTags:Gameplay Tag Container Structure)
     Tag:Gameplay Tag Structure=@n4
     bExactMatch:bool=true
    ) -> (ReturnValue:bool)
   )
   > Remove Item!(execute, TargetArray:Array of Gameplay Tag Structures=Get TagTracking(), Item:"Gameplay Tag Structure (by ref)"=@n4) -> (then, ReturnValue:bool)
   Remove!(
    execute=@n9.else
    TargetMap:Map of Gameplay Tag Structures to Sound Base Soft Object References=Get Sounds(self=Get SoundLibrary())
    Key:"Gameplay Tag Structure (by ref)"=@n4
   ) -> (then, ReturnValue:bool)
   n10 = Get EntryObjectCasted()
   > Remove Item(
    TargetArray:Array of Object References=Get SelectedEntries()
    Item:"Object Reference (by ref)"=Find(
     TargetMap:Map of Sound Library Tag Structures to Sound Library Editor Library Entry Object References=Get TagToLibraryEntry(self=@n10)
     Key:"Sound Library Tag Structure (by ref)"=@n4
    ) -> (Value:Sound Library Editor Library Entry Object Reference)
   )
  ]}
 TransactionAdd {nodes [
   n1 = TransactionAdd!() -> (then, Object:Object Reference) {is=title K2Node_FunctionEntry}
   > TransactObject(Object=@n1)
  ]}
 TransactionEnd {nodes [
   TransactionEnd() {is=title K2Node_FunctionEntry}
   > EndTransaction()
  ]}
 TransactionStart {nodes [
   n1 = TransactionStart!() -> (then, Description:Text) {is=title K2Node_FunctionEntry}
   n2 = Self-Reference()
   > BeginTransaction(Context=SoundLibraryEditor_EditEntry, Description=@n1, PrimaryObject=@n2)
   > TransactObject(Object=@n2)
  ]}
]
variables [
 EntryObject {type=Object, default=null}
 EntryObjectCasted {type=SoundLibraryEditor_EditEntry_C, default=null}
 SelectionTags {type="array<GameplayTag>", default []}
 Sound {type=SoundBase, default=null}
 SoundLibraryTags {type=GameplayTagContainer, default {gameplayTags []}}
]
widgets {
 info {inheritedWidgetCount=0, parentClass=&EditorUtilityWidget, rootWidgetClass=&CanvasPanel, widgetCount=5}
 widgets [
  CanvasPanel_178 {widgetClassPath=&CanvasPanel, widget=CanvasPanel_178}
  Border_2 {
   widgetClassPath=&Border
   parent=CanvasPanel_178
   widget=Border_2
   properties {
    background {
     bIsDynamicallyLoaded=false
     drawAs=Box
     imageType=NoImage
     mirroring=NoMirror
     resourceName=None
     resourceObject=null
     tiling=NoTile
     imageSize=<32 32>
     margin=<ltrb 0 0 0 0>
     outlineSettings {
      bUseBrushTransparency=false
      roundingType=HalfHeightRadius
      width=0
      color {colorUseRule=UseColor_Specified, specifiedColor=<rgba 0 0 0 0>}
      cornerRadii=<0 0 0 0>
     }
     tintColor {colorUseRule=UseColor_Specified, specifiedColor=<rgba 1 1 1 1>}
     uVRegion {bIsValid=false, max=<0 0>, min=<0 0>}
    }
    brushColor=<rgba 0 0 0 0.2>
   }
   slot {
    class=CanvasPanelSlot
    properties {bAutoSize=true, layoutData {alignment=<0 0>, anchors {maximum=<0 0>, minimum=<0 0>}, offsets=<ltrb 0 0 350 169.5912>}}
   }
  }
  VerticalBox_142 {widgetClassPath=&VerticalBox, parent=Border_2, widget=VerticalBox_142, slot {class=BorderSlot}}
  Tags {
   widgetClassPath=&SinglePropertyView
   parent=VerticalBox_142
   widget=Tags
   properties {nameOverride=" ", propertyName=SoundLibraryTags}
   slot {class=VerticalBoxSlot}
  }
  Sounds {
   widgetClassPath=&SinglePropertyView
   parent=VerticalBox_142
   widget=Sounds
   properties {nameOverride=" ", propertyName=Sound}
   slot {class=VerticalBoxSlot}
  }
 ]
}
