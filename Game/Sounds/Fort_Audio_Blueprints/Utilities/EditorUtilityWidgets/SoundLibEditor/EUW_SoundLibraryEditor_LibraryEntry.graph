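# Editor utility list row for the Sound Library Editor, showing one gameplay tag and its sound asset.
# Refresh Display shows the tag's leaf name indented by depth, dims soundless entries, and shows a collapse arrow on Transient entries, checked via the parent's TagsToIgnore.
# Clicks and collapse toggles forward to the parent editor.
# An invisible SinglePropertyView overlay edits Sound; changes clear Transient and write to the library's tag-to-sound map inside a transaction.
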WidgetBlueprint /Game/Sounds/Fort_Audio_Blueprints/Utilities/EditorUtilityWidgets/SoundLibEditor/EUW_SoundLibraryEditor_LibraryEntry
graphs [
 EventGraph {nodes [
   n1 = "On Check State Changed (CollapseCheckbox)"!() -> (OutputDelegate, then, bIsChecked:bool) {is=title K2Node_ComponentBoundEvent}
   > OnEntryCollapsed!(
    execute
    self:EUW Sound Library Editor Object Reference=Get Parent() -> (Parent:EUW Sound Library Editor Object Reference)
    Widget:EUW Sound Library Editor Library Entry Object Reference=Self-Reference()
    Collapsed:bool=@n1.bIsChecked
   ) -> (then)
   > Refresh Display()
   "On Property Changed (SingleProperty)"!() -> (OutputDelegate, then, PropertyName:Name=None) {is=title K2Node_ComponentBoundEvent}
   > OnSoundPropertyChanged()
   Refresh Display() {is=title K2Node_CustomEvent}
   n2 = GetDebugStringFromGameplayTag!(GameplayTag:Gameplay Tag Structure=Get Tag!(self:Sound Library Editor Library Entry Object Reference=Get EntryObjectCasted() -> (EntryObjectCasted:Sound Library Editor Library Entry Object Reference)) -> (Tag:Sound Library Tag Structure)) -> (ReturnValue:String)
   n3 = Split(SourceString=@n2, InStr=., SearchDir=FromEnd)
   > "SetText (Text)"(
    self=Get TextTag() -> (TextTag:Text Block Object Reference)
    InText="To Text (String)"(InString=SelectString(A=@n3.RightS, B=@n2, bPickA=@n3.ReturnValue))
   )
   n4 = int - int(A=Length(TargetArray:Array of Strings=ParseIntoArray(SourceString=@n2, Delimiter=.)))
   > SetPadding!(
    execute
    self:Border Object Reference=Get Border_98() -> (Border_98:Border Object Reference)
    InPadding_Left:float=float * float(A="To Float (Integer)"(InInt=@n4), B=20) {is=title K2Node_PromotableOperator}
    InPadding_Top:float=0
    InPadding_Right:float=0
    InPadding_Bottom:float=0
   ) -> (then)
   n5 = Get TextTag()
   n6 = Break Slate Color!(SlateColor:"Slate Color Structure (by ref)"=Get ColorAndOpacity!(self:Text Object Reference=@n5) -> (ColorAndOpacity:Slate Color Structure)) -> (
    SpecifiedColor_R:float=~
    SpecifiedColor_G:float=~
    SpecifiedColor_B:float=~
    SpecifiedColor_A:float=~
    ColorUseRule:ESlateColorStylingMode Enum
   ) {is=title K2Node_BreakStruct}
   n7 = IsValidSoftObjectReference(SoftObjectReference=Get Sound!(self:Sound Library Editor Library Entry Object Reference=Get EntryObjectCasted()) -> (Sound:Sound Base Soft Object Reference))
   > SetColorAndOpacity(
    self=@n5
    InColorAndOpacity=Make Slate Color!(
     SpecifiedColor_R:float=@n6
     SpecifiedColor_G:float=@n6.SpecifiedColor_G
     SpecifiedColor_B:float=@n6.SpecifiedColor_B
     SpecifiedColor_A:float=SelectFloat(A=1, B=0.5, bPickA=@n7)
     ColorUseRule=@n6.ColorUseRule
    ) -> (SlateColor)
   )
   n8 = Get Sound(self=Get EntryObjectCasted())
   n9 = GetSoftPtrAssetName!(Asset:Soft Object Reference=Get Sound(self=Get EntryObjectCasted())) -> (ReturnValue:String)
   > "SetText (Text)"(
    self=Get TextSoundAsset() -> (TextSoundAsset:Text Block Object Reference)
    InText=Select(Option 0:Text, Option 1:Text="To Text (String)"(InString=@n9), Index=IsValidSoftObjectReference(SoftObjectReference=@n8)) -> (ReturnValue:Text)
   )
   n10 = Get EntryObjectCasted()
   > SetVisibility(
    self=Get CollapseCheckbox() -> (CollapseCheckbox:Check Box Object Reference)
    InVisibility=Select(
     Option 0:ESlateVisibility Enum=Collapsed
     Option 1:ESlateVisibility Enum=Visible
     Index=Get Transient!(self:Sound Library Editor Library Entry Object Reference=@n10) -> (Transient:bool)
    ) -> (ReturnValue:ESlateVisibility Enum)
   )
   n11 = IsCollapsed!() -> (ReturnValue:bool)
   > SetRenderTransformAngle!(execute, self:Widget Object Reference=Get CollapseCheckbox(), Angle:float=SelectFloat(A=0, B=90, bPickA=@n11)) -> (then)
   > SetCheckedState!(
    execute
    self:Check Box Object Reference=Get CollapseCheckbox()
    InCheckedState:ECheckBoxState Enum=Select(Option 0:ECheckBoxState Enum=Unchecked, Option 1:ECheckBoxState Enum=Checked, Index=@n11) -> (ReturnValue:ECheckBoxState Enum)
   ) -> (then)
   n12 = Event OnListItemObjectSet()
   > n13 = Set EntryObject(EntryObject:Object Reference=@n12.ListItemObject) -> (Output_Get:Object Reference)
   > n14 = Cast To SoundLibraryEditor_LibraryEntry!(execute, Object=@n13) -> (then, CastFailed, AsSound Library Editor Library Entry:Sound Library Editor Library Entry Object Reference) {is=title K2Node_DynamicCast}
   > n15 = Set EntryObjectCasted(EntryObjectCasted:Sound Library Editor Library Entry Object Reference=@n14) -> (Output_Get:Sound Library Editor Library Entry Object Reference)
   > Set Widget!(
    execute
    Widget:EUW Sound Library Editor Library Entry Object Reference=Self-Reference()
    self:Sound Library Editor Library Entry Object Reference=@n15
   ) -> (then, Output_Get:EUW Sound Library Editor Library Entry Object Reference)
   > SetObject(self=Get SingleProperty() -> (SingleProperty:Single Property View Object Reference), NewObject=@n14)
   > Refresh Display() {is=title K2Node_CallFunction}
  ]}
 IsCollapsed {nodes [
   IsCollapsed() {is=title K2Node_FunctionEntry}
   > n1 = Get!(execute:Exec) -> (then:Exec, else:Exec, Parent:EUW Sound Library Editor Object Reference)
   n2 = GetDebugStringFromGameplayTag(GameplayTag=Get Tag(self=Get EntryObjectCasted()))
   > Return Node!(
    execute
    ReturnValue:bool=Contains Item(
     TargetArray:Array of Strings=Get TagsToIgnore!(self:EUW Sound Library Editor Object Reference=@n1) -> (TagsToIgnore:Array of Strings)
     ItemToFind:"String (by ref)"=Split(SourceString=@n2, InStr=., SearchDir=FromEnd).RightS
    )
   )
   Return Node(execute=@n1.else, ReturnValue=false)
  ]}
 OnClicked {nodes [
   OnClicked() {is=title K2Node_FunctionEntry}
   Return Node!(execute)
  ]}
 OnMouseButtonDown_0 {nodes [
   OnMouseButtonDown_0!() -> (then, MyGeometry:Geometry Structure, MouseEvent:"Pointer Event Structure (by ref)") {is=title K2Node_FunctionEntry}
   > n1 = Branch(Condition=Get Transient(self=Get EntryObjectCasted()))
   OnEntryClicked!(
    execute=@n1.else
    self:EUW Sound Library Editor Object Reference=Get Parent()
    Widget:EUW Sound Library Editor Library Entry Object Reference=Self-Reference()
    DoubleClick:bool=false
    ForceAdditiveSelect:bool=false
   ) -> (then)
   > Return Node!(execute, ReturnValue:Event Reply Structure=Make Event Reply!() -> (EventReply:Event Reply Structure) {is=title K2Node_MakeStruct})
  ]}
 OnMouseDoubleClick {nodes [
   OnMouseDoubleClick!() -> (then, MyGeometry:Geometry Structure, MouseEvent:"Pointer Event Structure (by ref)") {is=title K2Node_FunctionEntry}
   > OnEntryClicked(self=Get Parent(), Widget=Self-Reference(), DoubleClick=true, ForceAdditiveSelect=false)
   > Return Node(ReturnValue=Make Event Reply())
  ]}
 OnPostInitialize {nodes [
   n1 = OnPostInitialize!() -> (then, Parent:EUW Sound Library Editor Object Reference) {is=title K2Node_FunctionEntry}
   > Set Parent(Parent:EUW Sound Library Editor Object Reference=@n1) -> (Output_Get:EUW Sound Library Editor Object Reference)
  ]}
 OnSelectionChanged {nodes [
   n1 = OnSelectionChanged!() -> (then, Selected:bool=~) {is=title K2Node_FunctionEntry}
   n2 = Get Border_56() -> (Border_56:Border Object Reference)
   n3 = Get!(self:Border Object Reference=@n2) -> (ContentColorAndOpacity_R:float=~, ContentColorAndOpacity_G:float=~, ContentColorAndOpacity_B:float=~, ContentColorAndOpacity_A:float=~)
   > SetBrushColor!(
    execute
    self=@n2
    InBrushColor_R:float=@n3
    InBrushColor_G:float=@n3.ContentColorAndOpacity_G
    InBrushColor_B:float=@n3.ContentColorAndOpacity_B
    InBrushColor_A:float=SelectFloat(A=0.1, B=0, bPickA=@n1)
   ) -> (then)
  ]}
 OnSoundPropertyChanged {nodes [
   OnSoundPropertyChanged() {is=title K2Node_FunctionEntry}
   > Set Transient!(execute, Transient:bool=false, self:Sound Library Editor Library Entry Object Reference=Get EntryObjectCasted()) -> (then, Output_Get:bool)
   > BeginTransaction(Context=SoundLibraryEditor_LibraryEntry, Description="To Text (String)"(InString=OnSoundChanged), PrimaryObject=Self-Reference())
   > n1 = Get!(execute:Exec) -> (then:Exec, else:Exec, Parent:EUW Sound Library Editor Object Reference)
   n2 = Get Library!(self:EUW Sound Library Editor Object Reference=@n1) -> (Library:Sound Library Object Reference)
   > n3 = Is Valid(InputObject=@n2)
   > TransactObject(Object=@n2)
   n4 = Get EntryObjectCasted()
   > n5 = Add!(
    execute
    TargetMap:Map of Gameplay Tag Structures to Sound Base Soft Object References=Get Sounds!(self:Sound Library Object Reference=@n2) -> (Sounds:Map of Gameplay Tag Structures to Sound Base Soft Object References)
    Key:"Gameplay Tag Structure (by ref)"=Get Tag(self=@n4)
    Value:"Sound Base Soft Object Reference (by ref)"=Get Sound(self=@n4)
   ) -> (then) {is=title K2Node_CallFunction}
   Refresh Display(execute=[@n5.then, @n3.Is Not Valid])
   > RefreshSoundPropertyList(self:EUW Sound Library Editor Object Reference=@n1)
   > EndTransaction()
  ]}
]
variables [
 EntryObject {type=Object, default=null}
 EntryObjectCasted {type=SoundLibraryEditor_LibraryEntry_C, default=null}
 In Padding {type=Margin, default=<ltrb 0 0 0 0>}
 Parent {type=EUW_SoundLibraryEditor_C, default=null}
]
widgets {
 info {inheritedWidgetCount=0, parentClass=&EditorUtilityWidget, rootWidgetClass=&CanvasPanel, widgetCount=14}
 widgets [
  CanvasPanel_178 {widgetClassPath=&CanvasPanel, widget=CanvasPanel_178}
  VerticalBox_142 {
   widgetClassPath=&VerticalBox
   parent=CanvasPanel_178
   widget=VerticalBox_142
   properties {visibility=Visible}
   slot {
    class=CanvasPanelSlot
    properties {bAutoSize=true, layoutData {alignment=<0 0>, anchors {maximum=<1 1>, minimum=<0 0>}, offsets=<ltrb 0 0 0 0>}}
   }
  }
  Overlay_156 {widgetClassPath=&Overlay, parent=VerticalBox_142, widget=Overlay_156, slot {class=VerticalBoxSlot}}
  HorizontalBox_82 {
   widgetClassPath=&HorizontalBox
   parent=Overlay_156
   widget=HorizontalBox_82
   properties {visibility=Visible}
   slot {class=OverlaySlot, properties {horizontalAlignment=HAlign_Fill, verticalAlignment=VAlign_Fill}}
  }
  Border_56 {
   widgetClassPath=&Border
   bIsVariable=true
   parent=HorizontalBox_82
   widget=Border_56
   properties {brushColor=<rgba 0.0242 0.0242 0.0242 0>, padding=<ltrb 0 0 0 0>}
   slot {class=HorizontalBoxSlot, properties {size {sizeRule=Fill, value=1}}}
  }
  GridPanel_0 {
   widgetClassPath=&GridPanel
   parent=Border_56
   widget=GridPanel_0
   properties {columnFill [0.4,0.6]}
   slot {class=BorderSlot, properties {padding=<ltrb 0 0 0 0>}}
  }
  Border_98 {
   widgetClassPath=&Border
   bIsVariable=true
   parent=GridPanel_0
   widget=Border_98
   properties {brushColor=<rgba 0.0242 0.0242 0.0242 0>, padding=<ltrb 1 1 1 1>}
   slot {class=GridSlot, properties {columnSpan=1}}
  }
  HorizontalBox_0 {
   widgetClassPath=&HorizontalBox
   parent=Border_98
   widget=HorizontalBox_0
   slot {class=BorderSlot, properties {padding=<ltrb 1 1 1 1>}}
  }
  CollapseCheckbox {
   widgetClassPath=/Script/UMG.CheckBox
   bIsVariable=true
   parent=HorizontalBox_0
   widget=CollapseCheckbox
   properties {
    renderTransform {angle=90, scale=<1 1>, shear=<0 0>, translation=<0 1>}
    renderTransformPivot=<0.45 0.5>
    widgetStyle {
     checkBoxType=CheckBox
     backgroundHoveredImage {
      bIsDynamicallyLoaded=false
      drawAs=NoDrawType
      imageType=NoImage
      mirroring=NoMirror
      resourceName=None
      resourceObject=null
      tiling=NoTile
      imageSize=<0 0>
      margin=<ltrb 0 0 0 0>
      outlineSettings {
       bUseBrushTransparency=false
       roundingType=HalfHeightRadius
       width=0
       color {colorUseRule=UseColor_Specified, specifiedColor=<rgba 0 0 0 0>}
       cornerRadii=<0 0 0 0>
      }
      tintColor {colorUseRule=UseColor_Specified, specifiedColor=<rgba 1 1 1 1>}
      uVRegion {bIsValid=false, max=<0 0>, min=<0 0>}
     }
     backgroundImage {
      bIsDynamicallyLoaded=false
      drawAs=NoDrawType
      imageType=NoImage
      mirroring=NoMirror
      resourceName=None
      resourceObject=null
      tiling=NoTile
      imageSize=<0 0>
      margin=<ltrb 0 0 0 0>
      outlineSettings {
       bUseBrushTransparency=false
       roundingType=HalfHeightRadius
       width=0
       color {colorUseRule=UseColor_Specified, specifiedColor=<rgba 0 0 0 0>}
       cornerRadii=<0 0 0 0>
      }
      tintColor {colorUseRule=UseColor_Specified, specifiedColor=<rgba 1 1 1 1>}
      uVRegion {bIsValid=false, max=<0 0>, min=<0 0>}
     }
     backgroundPressedImage {
      bIsDynamicallyLoaded=false
      drawAs=NoDrawType
      imageType=NoImage
      mirroring=NoMirror
      resourceName=None
      resourceObject=null
      tiling=NoTile
      imageSize=<0 0>
      margin=<ltrb 0 0 0 0>
      outlineSettings {
       bUseBrushTransparency=false
       roundingType=HalfHeightRadius
       width=0
       color {colorUseRule=UseColor_Specified, specifiedColor=<rgba 0 0 0 0>}
       cornerRadii=<0 0 0 0>
      }
      tintColor {colorUseRule=UseColor_Specified, specifiedColor=<rgba 1 1 1 1>}
      uVRegion {bIsValid=false, max=<0 0>, min=<0 0>}
     }
     borderBackgroundColor {colorUseRule=UseColor_Specified, specifiedColor=<rgba 1 1 1 1>}
     checkedForeground {colorUseRule=UseColor_Foreground, specifiedColor=<rgba 1 0 1 1>}
     checkedHoveredForeground {colorUseRule=UseColor_Foreground, specifiedColor=<rgba 1 0 1 1>}
     checkedHoveredImage {
      bIsDynamicallyLoaded=false
      drawAs=Image
      imageType=FullColor
      mirroring=NoMirror
      resourceName=../../../Engine/Content/Slate/Common/CheckBox.png
      resourceObject=/FortUILibrary/Foundation/Buttons/Materials/Arrows/M_Button_CycleArrow_Right_Hovered
      tiling=NoTile
      imageSize=<16 16>
      margin=<ltrb 0 0 0 0>
      outlineSettings {
       bUseBrushTransparency=false
       roundingType=HalfHeightRadius
       width=0
       color {colorUseRule=UseColor_Specified, specifiedColor=<rgba 0 0 0 0>}
       cornerRadii=<0 0 0 0>
      }
      tintColor {colorUseRule=UseColor_Specified, specifiedColor=<rgba 1 1 1 1>}
      uVRegion {bIsValid=false, max=<0 0>, min=<0 0>}
     }
     checkedImage {
      bIsDynamicallyLoaded=false
      drawAs=Image
      imageType=FullColor
      mirroring=NoMirror
      resourceName=../../../Engine/Content/Slate/Common/CheckBox.png
      resourceObject=/FortUILibrary/Foundation/Buttons/Materials/Arrows/M_Button_CycleArrow_Right_Base
      tiling=NoTile
      imageSize=<16 16>
      margin=<ltrb 0 0 0 0>
      outlineSettings {
       bUseBrushTransparency=false
       roundingType=HalfHeightRadius
       width=0
       color {colorUseRule=UseColor_Specified, specifiedColor=<rgba 0 0 0 0>}
       cornerRadii=<0 0 0 0>
      }
      tintColor {colorUseRule=UseColor_Specified, specifiedColor=<rgba 1 1 1 1>}
      uVRegion {bIsValid=false, max=<0 0>, min=<0 0>}
     }
     checkedPressedForeground {colorUseRule=UseColor_Foreground, specifiedColor=<rgba 1 0 1 1>}
     checkedPressedImage {
      bIsDynamicallyLoaded=false
      drawAs=Image
      imageType=FullColor
      mirroring=NoMirror
      resourceName=../../../Engine/Content/Slate/Common/CheckBox.png
      resourceObject=/FortUILibrary/Foundation/Buttons/Materials/Arrows/M_Button_CycleArrow_Right_Pressed
      tiling=NoTile
      imageSize=<16 16>
      margin=<ltrb 0 0 0 0>
      outlineSettings {
       bUseBrushTransparency=false
       roundingType=HalfHeightRadius
       width=0
       color {colorUseRule=UseColor_Specified, specifiedColor=<rgba 0 0 0 0>}
       cornerRadii=<0 0 0 0>
      }
      tintColor {colorUseRule=UseColor_Specified, specifiedColor=<rgba 1 1 1 1>}
      uVRegion {bIsValid=false, max=<0 0>, min=<0 0>}
     }
     checkedSlateSound {resourceObject=null}
     foregroundColor {colorUseRule=UseColor_Foreground, specifiedColor=<rgba 1 0 1 1>}
     hoveredForeground {colorUseRule=UseColor_Foreground, specifiedColor=<rgba 1 0 1 1>}
     hoveredSlateSound {resourceObject=null}
     padding=<ltrb 2 0 0 0>
     pressedForeground {colorUseRule=UseColor_Foreground, specifiedColor=<rgba 1 0 1 1>}
     uncheckedHoveredImage {
      bIsDynamicallyLoaded=false
      drawAs=Image
      imageType=FullColor
      mirroring=NoMirror
      resourceName=../../../Engine/Content/Slate/Common/CheckBox.png
      resourceObject=&M_Button_CycleArrow_Right_Hovered
      tiling=NoTile
      imageSize=<16 16>
      margin=<ltrb 0 0 0 0>
      outlineSettings {
       bUseBrushTransparency=false
       roundingType=HalfHeightRadius
       width=0
       color {colorUseRule=UseColor_Specified, specifiedColor=<rgba 0 0 0 0>}
       cornerRadii=<0 0 0 0>
      }
      tintColor {colorUseRule=UseColor_Specified, specifiedColor=<rgba 1 1 1 1>}
      uVRegion {bIsValid=false, max=<0 0>, min=<0 0>}
     }
     uncheckedImage {
      bIsDynamicallyLoaded=false
      drawAs=Image
      imageType=FullColor
      mirroring=NoMirror
      resourceName=../../../Engine/Content/Slate/Common/CheckBox.png
      resourceObject=&M_Button_CycleArrow_Right_Base
      tiling=NoTile
      imageSize=<16 16>
      margin=<ltrb 0 0 0 0>
      outlineSettings {
       bUseBrushTransparency=false
       roundingType=HalfHeightRadius
       width=0
       color {colorUseRule=UseColor_Specified, specifiedColor=<rgba 0 0 0 0>}
       cornerRadii=<0 0 0 0>
      }
      tintColor {colorUseRule=UseColor_Specified, specifiedColor=<rgba 1 1 1 1>}
      uVRegion {bIsValid=false, max=<0 0>, min=<0 0>}
     }
     uncheckedPressedImage {
      bIsDynamicallyLoaded=false
      drawAs=Image
      imageType=FullColor
      mirroring=NoMirror
      resourceName=../../../Engine/Content/Slate/Common/CheckBox.png
      resourceObject=&M_Button_CycleArrow_Right_Pressed
      tiling=NoTile
      imageSize=<16 16>
      margin=<ltrb 0 0 0 0>
      outlineSettings {
       bUseBrushTransparency=false
       roundingType=HalfHeightRadius
       width=0
       color {colorUseRule=UseColor_Specified, specifiedColor=<rgba 0 0 0 0>}
       cornerRadii=<0 0 0 0>
      }
      tintColor {colorUseRule=UseColor_Specified, specifiedColor=<rgba 1 1 1 1>}
      uVRegion {bIsValid=false, max=<0 0>, min=<0 0>}
     }
     uncheckedSlateSound {resourceObject=null}
     undeterminedForeground {colorUseRule=UseColor_Foreground, specifiedColor=<rgba 1 0 1 1>}
     undeterminedHoveredImage {
      bIsDynamicallyLoaded=false
      drawAs=Image
      imageType=FullColor
      mirroring=NoMirror
      resourceName=../../../Engine/Content/Slate/Common/CheckBox_Undetermined_Hovered.png
      resourceObject=null
      tiling=NoTile
      imageSize=<16 16>
      margin=<ltrb 0 0 0 0>
      outlineSettings {
       bUseBrushTransparency=false
       roundingType=HalfHeightRadius
       width=0
       color {colorUseRule=UseColor_Specified, specifiedColor=<rgba 0 0 0 0>}
       cornerRadii=<0 0 0 0>
      }
      tintColor {colorUseRule=UseColor_Specified, specifiedColor=<rgba 1 1 1 1>}
      uVRegion {bIsValid=false, max=<0 0>, min=<0 0>}
     }
     undeterminedImage {
      bIsDynamicallyLoaded=false
      drawAs=Image
      imageType=FullColor
      mirroring=NoMirror
      resourceName=../../../Engine/Content/Slate/Common/CheckBox_Undetermined.png
      resourceObject=null
      tiling=NoTile
      imageSize=<16 16>
      margin=<ltrb 0 0 0 0>
      outlineSettings {
       bUseBrushTransparency=false
       roundingType=HalfHeightRadius
       width=0
       color {colorUseRule=UseColor_Specified, specifiedColor=<rgba 0 0 0 0>}
       cornerRadii=<0 0 0 0>
      }
      tintColor {colorUseRule=UseColor_Specified, specifiedColor=<rgba 1 1 1 1>}
      uVRegion {bIsValid=false, max=<0 0>, min=<0 0>}
     }
     undeterminedPressedImage {
      bIsDynamicallyLoaded=false
      drawAs=Image
      imageType=FullColor
      mirroring=NoMirror
      resourceName=../../../Engine/Content/Slate/Common/CheckBox_Undetermined_Hovered.png
      resourceObject=null
      tiling=NoTile
      imageSize=<16 16>
      margin=<ltrb 0 0 0 0>
      outlineSettings {
       bUseBrushTransparency=false
       roundingType=HalfHeightRadius
       width=0
       color {colorUseRule=UseColor_Specified, specifiedColor=<rgba 0 0 0 0>}
       cornerRadii=<0 0 0 0>
      }
      tintColor {colorUseRule=UseColor_Specified, specifiedColor=<rgba 0.5 0.5 0.5 1>}
      uVRegion {bIsValid=false, max=<0 0>, min=<0 0>}
     }
    }
   }
   slot {class=HorizontalBoxSlot, properties {padding=<ltrb 2.5 0 0 0>}}
  }
  TextTag {
   widgetClassPath=&UEFN_TextBlock
   bIsVariable=true
   parent=HorizontalBox_0
   widget=TextTag
   properties {
    clipping=ClipToBounds
    text=SoundMix Asset
    font {
     bForceMonospaced=false
     bMaterialIsStencil=false
     fontMaterial=null
     fontObject=&NotoSans
     letterSpacing=5
     monospacedWidth=1
     size=10
     skewAmount=0
     typefaceFontName=NotoSans-Regular
     outlineSettings {
      bApplyOutlineToDropShadows=false
      bMiteredCorners=false
      bSeparateFillAlpha=false
      outlineMaterial=null
      outlineSize=0
      outlineColor=<rgba 0 0 0 1>
     }
    }
   }
   slot {class=HorizontalBoxSlot, properties {horizontalAlignment=HAlign_Left, verticalAlignment=VAlign_Center, padding=<ltrb 2.5 0 0 0>}}
  }
  Border {
   widgetClassPath=&Border
   parent=GridPanel_0
   widget=Border
   properties {horizontalAlignment=HAlign_Left, verticalAlignment=VAlign_Center, brushColor=<rgba 0.0097 0.0097 0.0097 0>, padding=<ltrb 1 1 1 1>}
   slot {class=GridSlot, properties {column=1}}
  }
  TextSoundAsset {
   widgetClassPath=&UEFN_TextBlock
   bIsVariable=true
   parent=Border
   widget=TextSoundAsset
   properties {
    clipping=ClipToBounds
    font {
     bForceMonospaced=false
     bMaterialIsStencil=false
     fontMaterial=null
     fontObject=&NotoSans
     letterSpacing=5
     monospacedWidth=1
     size=10
     skewAmount=0
     typefaceFontName=NotoSans-Regular
     outlineSettings {
      bApplyOutlineToDropShadows=false
      bMiteredCorners=false
      bSeparateFillAlpha=false
      outlineMaterial=null
      outlineSize=0
      outlineColor=<rgba 0 0 0 1>
     }
    }
   }
   slot {class=BorderSlot, properties {horizontalAlignment=HAlign_Left, verticalAlignment=VAlign_Center, padding=<ltrb 1 1 1 1>}}
  }
  SoundBorder {
   widgetClassPath=&Border
   bIsVariable=true
   parent=Overlay_156
   widget=SoundBorder
   properties {
    brushColor=<rgba 0.0242 0.0242 0.0242 0>
    contentColorAndOpacity=<rgba 1 1 1 0>
    desiredSizeScale=<3 0.05>
    padding=<ltrb 0 0 0 0>
    renderTransform {angle=0, scale=<2 1>, shear=<0 0>, translation=<0 0>}
   }
   slot {class=OverlaySlot, properties {horizontalAlignment=HAlign_Fill, verticalAlignment=VAlign_Fill}}
  }
  SingleProperty {
   widgetClassPath=&SinglePropertyView
   parent=SoundBorder
   widget=SingleProperty
   properties {clipping=ClipToBounds, nameOverride=" ", propertyName=Sound, renderTransform {angle=0, scale=<1 1>, shear=<0 0.5>, translation=<0 0>}}
   slot {class=BorderSlot, properties {padding=<ltrb 0 0 0 0>}}
  }
 ]
}
