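# Editor utility widget for browsing, editing and auditioning a Sound Library asset that maps gameplay tags to sounds.
# It builds a collapsible tag tree, adding transient parent entries, and click, Ctrl and Shift selection fills a property list grouped by sound.
# Add and remove actions run inside undo transactions, and saving writes the library.
# Playback spawns spatialized sounds at a chosen distance in PIE or editor worlds, setting Local, Enemy or Friendly parameters.
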EditorUtilityWidgetBlueprint /Game/Sounds/Fort_Audio_Blueprints/Utilities/EditorUtilityWidgets/SoundLibEditor/EUW_SoundLibraryEditor
graphs [
 Action Add Tags {nodes [
   Action Add Tags() {is=title K2Node_FunctionEntry}
   n1 = Get CurrentActionObject() -> (CurrentActionObject:Object Reference)
   > n2 = For Each Loop(Array:Array of Gameplay Tag Structures=Get!(self:Sound Library Editor Action Add Multiple Tags Object Reference=Cast To SoundLibraryEditor_Action_AddMultipleTags!(Object=@n1) -> (AsSound Library Editor Action Add Multiple Tags:Sound Library Editor Action Add Multiple Tags Object Reference, bSuccess:bool=~) {is=title K2Node_DynamicCast}) -> (Tags_GameplayTags:Array of Gameplay Tag Structures)) -> (Array Element:Gameplay Tag Structure)
   > Add!(
    execute
    TargetMap:Map of Gameplay Tag Structures to Sound Base Soft Object References=Get Sounds!(self:Sound Library Object Reference=Get Library() -> (Library:Sound Library Object Reference)) -> (Sounds:Map of Gameplay Tag Structures to Sound Base Soft Object References)
    Key:"Gameplay Tag Structure (by ref)"=@n2
    Value:"Sound Base Soft Object Reference (by ref)"
   ) -> (then) {is=title K2Node_CallFunction}
  ]}
 Action_RemoveEntry {nodes [
   Action_RemoveEntry() {is=title K2Node_FunctionEntry}
   > n1 = For Each Loop(Array:Array of Object References=Get SelectedEntries() -> (SelectedEntries:Array of Object References)) -> (Array Element:Object Reference)
   > Remove!(
    execute
    TargetMap:Map of Gameplay Tag Structures to Sound Base Soft Object References=Get Sounds(self=Get Library())
    Key:"Gameplay Tag Structure (by ref)"=Get Tag!(self:Sound Library Editor Library Entry Object Reference=Cast To SoundLibraryEditor_LibraryEntry!(Object=@n1) -> (AsSound Library Editor Library Entry:Sound Library Editor Library Entry Object Reference, bSuccess:bool=~) {is=title K2Node_DynamicCast}) -> (Tag:Sound Library Tag Structure)
   ) -> (then, ReturnValue:bool)
  ]}
 EventGraph {nodes [
   n1 = "On Entry Initialized (SoundLibEntries)"!() -> (OutputDelegate, then, Item:Object Reference, Widget:User Widget Object Reference) {is=title K2Node_ComponentBoundEvent}
   > n2 = Cast To EUW_SoundLibraryEditor_LibraryEntry!(execute, Object=@n1.Widget) -> (then, CastFailed, AsEUW Sound Library Editor Library Entry:EUW Sound Library Editor Library Entry Object Reference) {is=title K2Node_DynamicCast}
   > OnPostInitialize!(execute, self:EUW Sound Library Editor Library Entry Object Reference=@n2, Parent:EUW Sound Library Editor Object Reference=Self-Reference()) -> (then)
   "On Property Changed (LibPropertyView)"!() -> (OutputDelegate, then, PropertyName:Name=None) {is=title K2Node_ComponentBoundEvent}
   > ClearListItems(self=Get SoundPropertyListView() -> (SoundPropertyListView:Common List View Object Reference))
   > Sort(self:Sound Library Object Reference=Get Library())
   > n3 = Get!(execute:Exec) -> (then:Exec, else:Exec, CachedLibrary:Sound Library Object Reference)
   > n4 = Unbind Event from On Rebuild Map(self:Sound Library Object Reference=@n3, Delegate=Create Event()) {is=title K2Node_RemoveDelegate}
   n5 = Set CachedLibrary(execute=[@n4.then, @n3.else], CachedLibrary:Sound Library Object Reference=Get Library()) -> (Output_Get:Sound Library Object Reference)
   > Bind Event to On Rebuild Map(self:Sound Library Object Reference=@n5, Delegate=Create Event()) {is=title K2Node_AddDelegate}
   > GenerateTagTree()
   n6 = "On Selection Changed (Distance_ComboBox)"!() -> (OutputDelegate, then, SelectedItem:String, SelectionType:ESelectInfo Enum) {is=title K2Node_ComponentBoundEvent}
   > Set PlaybackDistance(PlaybackDistance:double=String To Float(InString=@n6.SelectedItem)) -> (Output_Get:double)
   n7 = "On Check State Changed (PlaybackCollapseCheckbox)"!() -> (OutputDelegate, then, bIsChecked:bool) {is=title K2Node_ComponentBoundEvent}
   > SetVisibility(
    self=Get Border_Playback() -> (Border_Playback:Border Object Reference)
    InVisibility=Select(Option 0=Collapsed, Option 1=Visible, Index=@n7.bIsChecked)
   )
   > SetRenderTransformAngle!(
    execute
    self:Widget Object Reference=Get PlaybackCollapseCheckbox() -> (PlaybackCollapseCheckbox:Check Box Object Reference)
    Angle:float=Select(Option 0:float=0, Option 1:float=90, Index=@n7.bIsChecked) -> (ReturnValue:float)
   ) -> (then)
   "On Clicked (AddEntryButton)"() {is=title K2Node_ComponentBoundEvent}
   > SetCurrentAction!(
    execute
    self
    Class:Object Class Reference=&/SoundLibraryEditor_Action_AddMultipleTags.SoundLibraryEditor_Action_AddMultipleTags_C
    Type:SoundLibraryEditor_ActionTypes Enum=NewEnumerator0
   ) -> (then)
   "On Clicked (RemoveEntry)"() {is=title K2Node_ComponentBoundEvent}
   > SetCurrentAction(Type=NewEnumerator3)
   > PerformCurrentAction()
   n8 = "On Item Is Hovered Changed (SoundLibEntries)"!() -> (OutputDelegate, then, Item:Object Reference, bIsHovered:bool) {is=title K2Node_ComponentBoundEvent}
   > Set bIsEntryHovered(bIsEntryHovered:bool=@n8.bIsHovered) -> (Output_Get:bool)
   > n9 = Cast To SoundLibraryEditor_LibraryEntry!(execute:Exec, Object=@n8.Item) -> (then:Exec, CastFailed:Exec, AsSound Library Editor Library Entry)
   > SetVisibility(
    self=Get SoundBorder!(self:EUW Sound Library Editor Library Entry Object Reference=Get Widget!(self:Sound Library Editor Library Entry Object Reference=@n9) -> (Widget:EUW Sound Library Editor Library Entry Object Reference)) -> (SoundBorder:Border Object Reference)
    InVisibility=Select(Option 0:ESlateVisibility Enum=Visible, Option 1:ESlateVisibility Enum=Hidden, Index=@n8.bIsHovered) -> (ReturnValue:ESlateVisibility Enum)
   )
   "On Clicked (PlaySelected)"() {is=title K2Node_ComponentBoundEvent}
   > PlaySelectedSounds()
   "On Clicked (ConfirmButton)"() {is=title K2Node_ComponentBoundEvent}
   > PerformCurrentAction()
   > HideCurrentAction()
   "On Clicked (CancelButton)"() {is=title K2Node_ComponentBoundEvent}
   > HideCurrentAction()
   "On Clicked (SaveButton)"() {is=title K2Node_ComponentBoundEvent}
   > SaveLibrary()
   "On Clicked (StopSelected)"() {is=title K2Node_ComponentBoundEvent}
   > StopSounds()
   Event Destruct()
   > StopSounds()
   Event Construct()
   > GenerateTagTree()
   > SetObject(self=Get LibPropertyView() -> (LibPropertyView:Single Property View Object Reference), NewObject=Self-Reference())
  ]}
 FilterTagTree {nodes [
   FilterTagTree() {is=title K2Node_FunctionEntry}
   > n1 = For Each Loop(Array=Get AllEntryObjects() -> (AllEntryObjects:Array of Sound Library Editor Library Entry Object References))
   n2 = Get FilteredEntryObjects() -> (FilteredEntryObjects:Array of Sound Library Editor Library Entry Object References)
   Set List Items!(
    execute=@n1.Completed
    self:List View Object Reference=Get SoundLibEntries() -> (SoundLibEntries:List View Object Reference)
    InListItems:Array of Object References=@n2
   ) -> (then) {is=title K2Node_CallFunction}
   > n3 = For Each Loop(Array=Get SelectedEntries())
   > n4 = Branch(Condition=Contains Item(
     TargetArray:Array of Sound Library Editor Library Entry Object References=@n2
     ItemToFind:"Sound Library Editor Library Entry Object Reference (by ref)"=Cast To SoundLibraryEditor_LibraryEntry!(Object=@n3) -> (AsSound Library Editor Library Entry, bSuccess:bool=~)
    ))
   Remove Item!(execute=@n4.else, TargetArray:Array of Object References=Get SelectedEntries(), Item:"Object Reference (by ref)"=@n3) -> (then, ReturnValue:bool)
   n5 = Cast To SoundLibraryEditor_LibraryEntry!(execute:Exec=@n1.LoopBody, Object=@n1) -> (then:Exec, CastFailed:Exec, AsSound Library Editor Library Entry)
   > n6 = Branch(Condition=ShouldExcludeTag!(GameplayTag:Gameplay Tag Structure=Get Tag(self=@n5)) -> (Result:bool))
   Add(
    execute=@n6.else
    TargetArray:Array of Sound Library Editor Library Entry Object References=Get FilteredEntryObjects()
    NewItem:"Sound Library Editor Library Entry Object Reference (by ref)"=@n5
   ) {is=title K2Node_CallArrayFunction}
  ]}
 FindEntryByTag {nodes [
   n1 = FindEntryByTag!() -> (then, TargetArray:Array of Sound Library Editor Library Entry Object References, Tag:"Gameplay Tag Structure (by ref)") {is=title K2Node_FunctionEntry}
   > n2 = For Each Loop(Array:Array of Sound Library Editor Library Entry Object References=@n1) -> (Array Element:Sound Library Editor Library Entry Object Reference)
   Return Node!(execute=@n2.Completed, Entry:Sound Library Editor Library Entry Object Reference)
   > Branch(Condition=MatchesTag!(TagOne:Gameplay Tag Structure=Get Tag(self=@n2), TagTwo:Gameplay Tag Structure=@n1.Tag, bExactMatch:bool=true) -> (ReturnValue:bool))
   > Return Node(Entry=@n2)
  ]}
 GenerateTagTree {nodes [
   GenerateTagTree() {is=title K2Node_FunctionEntry}
   > ClearListItems(self=Get SoundLibEntries())
   > Clear(TargetArray:Array of Sound Library Editor Library Entry Object References=Get AllEntryObjects())
   > Clear(TargetArray:Array of Strings=Get TagsToIgnore() -> (TagsToIgnore:Array of Strings))
   > n1 = Get!(execute) -> (then, else, Library:Sound Library Object Reference)
   > Sort(self=@n1)
   n2 = Get Sounds(self=@n1)
   > n3 = Keys!(execute, TargetMap:Map of Gameplay Tag Structures to Sound Base Soft Object References=@n2) -> (then, Keys:Array of Gameplay Tag Structures)
   > Set LocalLibTags(LocalLibTags:Array of Gameplay Tag Structures=@n3) -> (Output_Get:Array of Gameplay Tag Structures)
   > n4 = Values!(execute, TargetMap:Map of Gameplay Tag Structures to Sound Base Soft Object References=@n2) -> (then, Values:Array of Sound Base Soft Object References)
   > Set LocalLibSounds(LocalLibSounds:Array of Sound Base Soft Object References=@n4) -> (Output_Get:Array of Sound Base Soft Object References)
   > n5 = For Each Loop(Array:Array of Gameplay Tag Structures=Get LocalLibTags() -> (LocalLibTags:Array of Gameplay Tag Structures)) -> (Array Element:Gameplay Tag Structure)
   Set List Items(execute=@n5.Completed, self=Get SoundLibEntries(), InListItems=Get AllEntryObjects())
   n6 = Sequence(execute=@n5.LoopBody)
   > n7 = GetArrayOfParentTags!(execute, self, Tag:Gameplay Tag Structure=@n5) -> (then, ParentTags:Array of Gameplay Tag Structures)
   > n8 = Reverse for Each Loop!(Exec:Exec, Array:Array of Gameplay Tag Structures=@n7) -> (LoopBody:Exec, ArrayIndex:int, ArrayElement:Gameplay Tag Structure, Completed:Exec) {is=title K2Node_MacroInstance}
   > n9 = Add Unique(
    TargetArray:Array of Gameplay Tag Structures=Get LocalParentTags() -> (LocalParentTags:Array of Gameplay Tag Structures)
    NewItem:"Gameplay Tag Structure (by ref)"=@n8.ArrayElement
   )
   > Branch(Condition="Not Equal (Integer)"(A=@n9, B=-1))
   n10 = Get AllEntryObjects()
   > n11 = Branch(Condition=IsValid(Object=FindEntryByTag!(TargetArray:Array of Sound Library Editor Library Entry Object References=@n10, Tag:"Gameplay Tag Structure (by ref)"=@n8.ArrayElement) -> (Entry:Sound Library Editor Library Entry Object Reference) {is=title K2Node_CallFunction}))
   n12 = Construct Object from Class(
    execute=@n11.else
    Class=&/SoundLibraryEditor_LibraryEntry.SoundLibraryEditor_LibraryEntry_C
   ) -> (ReturnValue:Sound Library Editor Library Entry Object Reference)
   > SetVariables!(
    execute
    self:Sound Library Editor Library Entry Object Reference=@n12
    Tag:Gameplay Tag Structure=@n8.ArrayElement
    Sound:Sound Base Soft Object Reference
    Transient:bool=true
   ) -> (then)
   > Add Unique(
    TargetArray:Array of Sound Library Editor Library Entry Object References=Get AllEntryObjects()
    NewItem:"Sound Library Editor Library Entry Object Reference (by ref)"=@n12
   )
   n13 = FindEntryByTag(TargetArray=Get AllEntryObjects(), Tag=@n5)
   n14 = Branch(execute=@n6.then_1, Condition=IsValid(Object=@n13))
   n15 = "Get (a copy)"(
    Array:Array of Sound Base Soft Object References=Get LocalLibSounds() -> (LocalLibSounds:Array of Sound Base Soft Object References)
    Dimension 1=@n5.Array Index
   ) -> (Output:Sound Base Soft Object Reference)
   > SetVariables(self=@n13, Tag=@n5, Sound=@n15, Transient=false)
   n16 = Construct Object from Class(
    execute=@n14.else
    Class=&SoundLibraryEditor_LibraryEntry_C
   )
   > SetVariables(self=@n16, Tag=@n5, Sound=@n15, Transient=false)
   > Add Unique(TargetArray=Get AllEntryObjects(), NewItem=@n16)
  ]}
 GetArrayOfParentTags {nodes [
   n1 = GetArrayOfParentTags!() -> (then, Tag:Gameplay Tag Structure) {is=title K2Node_FunctionEntry}
   > Set TagString(TagString:String=GetDebugStringFromGameplayTag!(GameplayTag:Gameplay Tag Structure=@n1) -> (ReturnValue:String)) -> (Output_Get:String)
   n2 = Get TagFound() -> (TagFound:bool)
   n3 = Get TagString() -> (TagString:String)
   > n4 = While Loop!(execute:Exec, Condition:bool=OR Boolean(A=Contains(SearchIn=@n3, Substring=.), B=NOT Boolean(A=@n2))) -> (LoopBody:Exec, Completed:Exec) {is=title K2Node_MacroInstance}
   Return Node!(execute=@n4.Completed, ParentTags:Array of Gameplay Tag Structures=Get Parents() -> (Parents:Array of Gameplay Tag Structures))
   n5 = Split(SourceString=Get TagString(), InStr=., SearchDir=FromEnd)
   > n6 = RequestGameplayTag!(execute, TagName:String=@n5, bErrorIfNotFound:bool=true) -> (then, OutTag:Gameplay Tag Structure, ReturnValue:bool)
   > n7 = Set TagFound(TagFound:bool=@n6.ReturnValue) -> (Output_Get:bool)
   > Branch(Condition=@n7)
   > Add(TargetArray:Array of Gameplay Tag Structures=Get Parents(), NewItem:"Gameplay Tag Structure (by ref)"=@n6)
   > Set TagString(TagString=@n5)
  ]}
 GetAudioLocation {nodes [
   GetAudioLocation() {is=title K2Node_FunctionEntry}
   > n1 = GetPIEWorlds!(execute, bIncludeDedicatedServer:bool=false) -> (then, ReturnValue:Array of World Object References)
   > n2 = Branch(Condition="integer > integer"(A=Length(TargetArray:Array of World Object References=@n1)))
   n3 = GetPlayerCharacter!(
    WorldContextObject:Object Reference="Get (a copy)"(Array:Array of World Object References=@n1) -> (Output:World Object Reference)
    PlayerIndex:int=0
   ) -> (ReturnValue:Character Object Reference)
   > n4 = Is Valid(InputObject=@n3)
   n5 = Set SpawnLocation(execute=@n4.Is Not Valid, SpawnLocation:Vector="0, 0, 0") -> (Output_Get:Vector)
   n6 = Branch(execute=[@n9.then, @n12.then, @n5.then], Condition=IsChecked(self=Get DrawDebug() -> (DrawDebug:Editor Utility Check Box Object Reference)))
   > n7 = DrawDebugSphere!(
    execute
    WorldContextObject:Object Reference=GetCurrentWorld!() -> (Object:Object Reference)
    Center:Vector=Get SpawnLocation() -> (SpawnLocation:Vector)
    Radius:float=200
    Segments:int=12
    LineColor:Linear Color Structure=(R=1,G=1,B=1,A=1)
    Duration:float=2
    Thickness:float=5
    DepthPriority:EDrawDebugSceneDepthPriorityGroup Enum=World
   ) -> (then)
   Return Node!(execute=[@n7.then, @n6.else], Location:Vector=Get SpawnLocation())
   n8 = Get Mesh(self=@n3)
   n9 = Set SpawnLocation(
    execute=@n4.Is Valid
    SpawnLocation=TransformLocation!(
     T_Location:Vector=Get World Location!(self:Scene Component Object Reference=@n8) -> (ReturnValue:Vector) {is=title K2Node_CallFunction}
     T_Rotation:Rotator="0, 0, 0"
     T_Scale:Vector="1,1,1"
     Location_X:double=Get PlaybackDistance() -> (PlaybackDistance:double)
     Location_Y:double=0
     Location_Z:double=0
    ) -> (ReturnValue:Vector)
   )
   n10 = GetLevelViewportCameraInfo!(self:Unreal Editor Subsystem Object Reference=Get UnrealEditorSubsystem()) -> (CameraLocation:Vector, CameraRotation:Rotator, ReturnValue:bool)
   n11 = Get PlaybackDistance()
   n12 = Set SpawnLocation(
    execute=@n2.else
    SpawnLocation=TransformLocation(T_Location=@n10, T_Rotation=@n10.CameraRotation, T_Scale="1,1,1", Location_X="Max (Float)"(A=@n11, B=1), Location_Y=0, Location_Z=0)
   )
  ]}
 GetCurrentWorld {nodes [
   GetCurrentWorld() {is=title K2Node_FunctionEntry}
   > n1 = GetPIEWorlds(bIncludeDedicatedServer=false)
   > n2 = Branch(Condition="integer > integer"(A=Length(TargetArray=@n1)))
   > n3 = Set WorldObject(WorldObject:Object Reference="Get (a copy)"(Array=@n1)) -> (Output_Get:Object Reference)
   Return Node!(execute=[@n3.then, @n5.then], Object:Object Reference=Get WorldObject() -> (WorldObject:Object Reference))
   n4 = GetEditorWorld(execute=@n2.else, self=Get UnrealEditorSubsystem())
   > n5 = Set WorldObject(WorldObject=@n4)
  ]}
 HideCurrentAction {nodes [
   HideCurrentAction() {is=title K2Node_FunctionEntry}
   > SetActiveWidgetIndex(self=Get WidgetSwitcher() -> (WidgetSwitcher:Widget Switcher Object Reference), Index=0)
  ]}
 OnEntryClicked {nodes [
   OnEntryClicked!() -> (then, Widget:EUW Sound Library Editor Library Entry Object Reference, DoubleClick:bool=~, ForceAdditiveSelect:bool=~) {is=title K2Node_FunctionEntry}
   n1 = GetModifierKeysState!() -> (ReturnValue:Modifier Keys State Structure)
   > Set isControlDown(isControlDown:bool=Is Control Down!(KeysState:"Modifier Keys State Structure (by ref)"=@n1) -> (ReturnValue:bool)) -> (Output_Get:bool)
   > Set isShiftDown(isShiftDown:bool=Is Shift Down!(KeysState:"Modifier Keys State Structure (by ref)"=@n1) -> (ReturnValue:bool)) -> (Output_Get:bool)
   > Set DisplayedEntryWidgets(DisplayedEntryWidgets:Array of Object References=GetDisplayedEntryWidgets!(self:List View Base Object Reference=Get SoundLibEntries()) -> (ReturnValue:Array of User Widget Object References)) -> (Output_Get:Array of Object References)
   > n2 = Sequence!(execute) -> (then_0, then_1, then_2:Exec)
   n3 = Branch(execute=@n2.then_1, Condition=Get isShiftDown() -> (isShiftDown:bool))
   > n4 = Get!(execute) -> (then, else, LastSelectedWidget:Object Reference)
   > n5 = Set StartInt(StartInt:int=Find Item!(
     TargetArray:Array of Object References=Get DisplayedEntryWidgets() -> (DisplayedEntryWidgets:Array of Object References)
     ItemToFind:"Object Reference (by ref)"=@n4
    ) -> (ReturnValue)) -> (Output_Get:int=0)
   n6 = Get!(execute=[@n11.then, @n5.then]) -> (then, else, Widget:EUW Sound Library Editor Library Entry Object Reference)
   > Set EndInt(EndInt:int=Find Item(TargetArray=Get DisplayedEntryWidgets(), ItemToFind=@n6)) -> (Output_Get:int=0)
   > n7 = For Each Loop(Array:Array of Object References=Get DisplayedEntryWidgets()) -> (Array Element:Object Reference)
   Set LastSelectedWidget(execute=@n7.Completed, LastSelectedWidget:Object Reference=Get Widget!() -> (Widget)) -> (Output_Get:Object Reference)
   n8 = Get StartInt() -> (StartInt:int=0)
   n9 = Get EndInt() -> (EndInt:int=0)
   Branch(
    execute=@n7.LoopBody
    Condition="In Range (Integer)"!(
     Value:int=@n7.Array Index
     Min:int="Min (Integer)"(A=@n8, B=@n9) -> (ReturnValue=0)
     Max:int="Max (Integer)"(A=@n8, B=@n9)
     InclusiveMin:bool=true
     InclusiveMax:bool=true
    ) -> (ReturnValue:bool)
   )
   > n10 = Cast To EUW_SoundLibraryEditor_LibraryEntry(Object=@n7)
   > OnSelectionChanged!(execute, self:EUW Sound Library Editor Library Entry Object Reference=@n10, Selected:bool=true) -> (then)
   > Add Unique(
    TargetArray:Array of Object References=Get SelectedEntries()
    NewItem:"Object Reference (by ref)"=Get EntryObject!(self:EUW Sound Library Editor Library Entry Object Reference=@n10) -> (EntryObject:Object Reference)
   )
   n11 = Set StartInt(execute=@n4.else, StartInt=0) -> (Output_Get=0)
   n12 = Get(execute=@n3.else)
   > OnSelectionChanged(self=@n12, Selected=true)
   > Set LastSelectedWidget(LastSelectedWidget=Get Widget())
   > Add Unique(TargetArray=Get SelectedEntries(), NewItem=Get EntryObject(self=Get Widget()))
   Set LastSelectedWidget(execute=@n12.else)
   RefreshSoundPropertyList(execute=@n2.then_2)
   n13 = Branch(
    execute=@n2.then_0
    Condition=OR Boolean!(A=Get isControlDown() -> (isControlDown:bool), B=Get isShiftDown(), C:bool=Get ForceAdditiveSelect() -> (ForceAdditiveSelect:bool)) -> (ReturnValue)
   )
   Clear(execute=@n13.else, TargetArray:Array of Object References=Get SelectedEntries())
   > n14 = For Each Loop(Array=Get DisplayedEntryWidgets())
   > n15 = Cast To EUW_SoundLibraryEditor_LibraryEntry(Object=@n14)
   > OnSelectionChanged(self=@n15, Selected=false)
  ]}
 OnEntryCollapsed {nodes [
   n1 = OnEntryCollapsed!() -> (then, Widget:EUW Sound Library Editor Library Entry Object Reference, Collapsed:bool=~) {is=title K2Node_FunctionEntry}
   > n2 = Cast To SoundLibraryEditor_LibraryEntry(Object=Get EntryObject(self=@n1))
   > n3 = Sequence!(execute) -> (then_0, then_1)
   > n4 = Branch(Condition=@n1.Collapsed)
   n5 = Split(SourceString=GetDebugStringFromGameplayTag(GameplayTag=Get Tag(self=@n2)), InStr=., SearchDir=FromEnd)
   n6 = Get TagsToIgnore()
   > Add(TargetArray:Array of Strings=@n6, NewItem:"String (by ref)"=@n5.RightS)
   Remove Item(execute=@n4.else, TargetArray:Array of Strings=@n6, Item:"String (by ref)"=@n5.RightS)
   FilterTagTree(execute=@n3.then_1) {is=title K2Node_CallFunction}
  ]}
 OnMouseButtonDown {nodes [
   OnMouseButtonDown!() -> (then, MyGeometry:Geometry Structure, MouseEvent:"Pointer Event Structure (by ref)") {is=title K2Node_FunctionEntry}
   > n1 = Branch(Condition=Get bIsEntryHovered() -> (bIsEntryHovered:bool))
   n2 = For Each Loop(Exec=@n1.else, Array=Get SelectedEntries())
   Clear(execute=@n2.Completed, TargetArray=Get SelectedEntries())
   > Return Node!(execute, ReturnValue:Event Reply Structure=Make Event Reply!() -> (EventReply:Event Reply Structure) {is=title K2Node_MakeStruct})
   OnSelectionChanged(
    execute=@n2.LoopBody
    self=Get Widget!(self:Sound Library Editor Library Entry Object Reference=Cast To SoundLibraryEditor_LibraryEntry!(Object=@n2) -> (AsSound Library Editor Library Entry, bSuccess:bool=~)) -> (Widget)
    Selected=false
   )
  ]}
 OnRebuildMap {nodes [
   OnRebuildMap() {is=title K2Node_FunctionEntry}
   > GenerateTagTree() {is=title K2Node_CallFunction}
   > OnEntryClicked!(execute, self, Widget:EUW Sound Library Editor Library Entry Object Reference, DoubleClick:bool=false, ForceAdditiveSelect:bool=false) -> (then) {is=title K2Node_CallFunction}
  ]}
 PerformCurrentAction {nodes [
   PerformCurrentAction() {is=title K2Node_FunctionEntry}
   n1 = Get CurrentActionType() -> (CurrentActionType:SoundLibraryEditor_ActionTypes Enum=NewEnumerator0)
   > TransactionStart!(
    execute
    self
    Description:Text="To Text (String)"(InString=Enum to String!(Enumerator:SoundLibraryEditor_ActionTypes Enum=@n1) -> (ReturnValue:String))
   ) -> (then)
   > TransactionAdd!(execute, self, Object:Object Reference=Get Library()) -> (then)
   > n2 = Sequence()
   > n3 = Switch on SoundLibraryEditor_ActionTypes!(execute, Selection:SoundLibraryEditor_ActionTypes Enum=Get CurrentActionType() -> (CurrentActionType=NewEnumerator0)) -> (NewEnumerator0:Exec, NewEnumerator3:Exec) {is=title K2Node_SwitchEnum}
   > Action Add Tags() {is=title K2Node_CallFunction}
   Action_RemoveEntry(execute=@n3.NewEnumerator3) {is=title K2Node_CallFunction}
   TransactionEnd(execute=@n2.then_1)
   > GenerateTagTree()
   > OnEntryClicked(DoubleClick=false, ForceAdditiveSelect=false)
  ]}
 PlaySelectedSounds {nodes [
   PlaySelectedSounds() {is=title K2Node_FunctionEntry}
   > StopSounds()
   > n1 = For Each Loop(Array=Get SelectedEntries())
   Set CurrentWorld(execute=@n1.Completed, CurrentWorld:Object Reference=GetCurrentWorld() {is=title K2Node_CallFunction}) -> (Output_Get:Object Reference)
   > n2 = For Each Loop(Array:Array of Sound Base Soft Object References=Get SoundsToPlay() -> (SoundsToPlay:Array of Sound Base Soft Object References)) -> (Array Element:Sound Base Soft Object Reference)
   > n3 = LoadAsset_Blocking(Asset=@n2)
   > n4 = Cast To SoundBase!(execute, Object=@n3) -> (then, CastFailed, AsSound Base:Sound Base Object Reference) {is=title K2Node_DynamicCast}
   > n5 = CreateSound2D!(
    execute
    WorldContextObject:Object Reference=Get CurrentWorld() -> (CurrentWorld:Object Reference)
    Sound:Sound Base Object Reference=@n4
    VolumeMultiplier:float=1
    PitchMultiplier:float=1
    StartTime:float=0
    ConcurrencySettings:Sound Concurrency Object Reference
    bPersistAcrossLevelTransition:bool=false
    bAutoDestroy:bool=true
   ) -> (then, ReturnValue:Audio Component Object Reference)
   > Set CurrentAudioComponent(CurrentAudioComponent:Audio Component Object Reference=@n5) -> (Output_Get:Audio Component Object Reference)
   > n6 = Get!(execute) -> (then, else, CurrentAudioComponent:Audio Component Object Reference)
   > Set World Location!(
    execute
    self:Scene Component Object Reference=@n6
    NewLocation:Vector=GetAudioLocation!() -> (Location:Vector) {is=title K2Node_CallFunction}
    bSweep:bool=false
    bTeleport:bool=true
   ) -> (then, SweepHitResult:Hit Result Structure) {is=title K2Node_CallFunction}
   n7 = Get CurrentAudioComponent() -> (CurrentAudioComponent:Audio Component Object Reference)
   > Set bAllowSpatialization!(execute, bAllowSpatialization:bool=true, self:Audio Component Object Reference=@n7) -> (then, Output_Get:bool)
   > Play!(execute, self:Audio Component Object Reference=@n7, StartTime:float=0) -> (then)
   > Add Unique(
    TargetArray:Array of Audio Component Object References=Get CurrentlyPlayingSounds() -> (CurrentlyPlayingSounds:Array of Audio Component Object References)
    NewItem:"Audio Component Object Reference (by ref)"=@n7
   )
   > n8 = Switch on String!(
    execute:Exec
    Selection:String=GetSelectedOption!(self:"Combo Box (String) Object Reference"=Get SoundAffiliation() -> (SoundAffiliation:Editor Utility Combo Box String Object Reference)) -> (ReturnValue:String)
   ) -> (Default:Exec, Local:Exec, Enemy:Exec, Friendly:Exec) {pinNames [Local,Enemy,Friendly]}
   n9 = Get CurrentAudioComponent()
   Set Boolean Parameter!(execute=@n8.Enemy, self:Audio Parameter Controller Interface Interface=@n9, InName:Name=isLocal, InBool:bool=false) -> (then) {is=title K2Node_CallFunction}
   > Set Integer Parameter!(execute, self:Audio Component Object Reference=@n9, InName:Name=PlayerAffiliation, InInt:int=2) -> (then) {is=title K2Node_CallFunction}
   > Set Boolean Parameter(self=@n9, InName=isFriendly, InBool=false)
   Set Boolean Parameter(execute=@n8.Friendly, self=@n9, InName=isLocal, InBool=false)
   > Set Integer Parameter(self=@n9, InName=PlayerAffiliation, InInt=1)
   > Set Boolean Parameter(self=@n9, InName=isFriendly, InBool=true)
   Set Boolean Parameter(execute=@n8.Local, self=@n9, InName=isLocal, InBool=true)
   > Set Integer Parameter(self=@n9, InName=PlayerAffiliation, InInt=0)
   > Set Boolean Parameter(self=@n9, InName=isFriendly, InBool=false)
   n10 = Cast To SoundLibraryEditor_LibraryEntry!(execute:Exec=@n1.LoopBody, Object=@n1) -> (then:Exec, CastFailed:Exec, AsSound Library Editor Library Entry)
   > Add Unique(
    TargetArray:Array of Sound Base Soft Object References=Get SoundsToPlay()
    NewItem:"Sound Base Soft Object Reference (by ref)"=Get Sound!(self:Sound Library Editor Library Entry Object Reference=@n10) -> (Sound:Sound Base Soft Object Reference)
   )
  ]}
 RefreshSoundPropertyList {nodes [
   RefreshSoundPropertyList() {is=title K2Node_FunctionEntry}
   > ClearListItems(self=Get SoundPropertyListView())
   > n1 = For Each Loop(Array:Array of Object References=Get SelectedEntries()) -> (Array Element:Object Reference)
   n2 = Values(
    execute=@n1.Completed
    TargetMap:Map of Sound Base Soft Object References to Sound Library Editor Edit Entry Object References=Get Entries() -> (Entries:Map of Sound Base Soft Object References to Sound Library Editor Edit Entry Object References)
   ) -> (Values:Array of Sound Library Editor Edit Entry Object References)
   > Set List Items(self=Get SoundPropertyListView(), InListItems=@n2)
   n3 = Length(TargetArray:Array of Object References=Get SelectedEntries())
   > SetVisibility(
    self=Get Playback() -> (Playback:Vertical Box Object Reference)
    InVisibility=Select(Option 0=Collapsed, Option 1=Visible, Index="integer > integer"(A=@n3))
   )
   n4 = Cast To SoundLibraryEditor_LibraryEntry(execute=@n1.LoopBody, Object=@n1)
   > n5 = Branch(Condition=Get Transient!(self:Sound Library Editor Library Entry Object Reference=@n4) -> (Transient:bool))
   n6 = Find(
    TargetMap:Map of Sound Base Soft Object References to Sound Library Editor Edit Entry Object References=Get Entries()
    Key:"Sound Base Soft Object Reference (by ref)"=Get Sound(self=@n4)
   ) -> (Value:Sound Library Editor Edit Entry Object Reference)
   n7 = Branch(execute=@n5.else, Condition=@n6.ReturnValue)
   > AddSoundLibraryEntryAndRefresh!(
    execute
    self:Sound Library Editor Edit Entry Object Reference=@n6
    SoundLibraryEntry:Sound Library Editor Library Entry Object Reference=@n4
   ) -> (then)
   n8 = Construct Object from Class(
    execute=@n7.else
    Class=&/SoundLibraryEditor_EditEntry.SoundLibraryEditor_EditEntry_C
   ) -> (ReturnValue:Sound Library Editor Edit Entry Object Reference)
   > SetVariables!(execute, self:Sound Library Editor Edit Entry Object Reference=@n8, Library:Sound Library Editor Library Entry Object Reference=@n4) -> (then)
   > Add(
    TargetMap:Map of Sound Base Soft Object References to Sound Library Editor Edit Entry Object References=Get Entries()
    Key:"Sound Base Soft Object Reference (by ref)"=Get Sound(self=@n4)
    Value:"Sound Library Editor Edit Entry Object Reference (by ref)"=@n8
   ) {is=title K2Node_CallFunction}
  ]}
 SaveLibrary {nodes [
   SaveLibrary() {is=title K2Node_FunctionEntry}
   > SaveLoadedAsset(AssetToSave=Get Library(), bOnlyIfIsDirty=true)
  ]}
 SetCurrentAction {nodes [
   n1 = SetCurrentAction!() -> (then, Class:Object Class Reference, Type:SoundLibraryEditor_ActionTypes Enum=NewEnumerator0) {is=title K2Node_FunctionEntry}
   > Set CurrentActionType(CurrentActionType:SoundLibraryEditor_ActionTypes Enum=@n1.Type) -> (Output_Get:SoundLibraryEditor_ActionTypes Enum=NewEnumerator0)
   > Branch(Condition=IsValidClass!(Class:Object Class Reference=@n1) -> (ReturnValue:bool))
   > n2 = Construct Object from Class(Class=@n1) -> (ReturnValue:Object Reference)
   > n3 = Set CurrentActionObject(CurrentActionObject:Object Reference=@n2) -> (Output_Get:Object Reference)
   > SetObject(self=Get ActionDetailsView() -> (ActionDetailsView:Details View Object Reference), NewObject=@n3)
   > SetActiveWidgetIndex(self=Get WidgetSwitcher(), Index=1)
   > Return Node!(execute)
  ]}
 ShouldExcludeTag {nodes [
   n1 = ShouldExcludeTag!() -> (then, GameplayTag:Gameplay Tag Structure) {is=title K2Node_FunctionEntry}
   > n2 = For Each Loop(Array:Array of Strings=Get TagsToIgnore()) -> (Array Element:String)
   Return Node!(execute=@n2.Completed, Result:bool=false)
   n3 = GetDebugStringFromGameplayTag(GameplayTag=@n1)
   n4 = ParseIntoArray(SourceString=@n3, Delimiter=.)
   n5 = Split(SourceString=@n3, InStr=., SearchDir=FromEnd)
   > Branch(Condition=AND Boolean(A="Not Equal Exactly (String)"(A=@n5.RightS, B=@n2), B=Contains Item(TargetArray:Array of Strings=@n4, ItemToFind:"String (by ref)"=@n2)))
   > Return Node(Result=true)
  ]}
 StopSounds {nodes [
   StopSounds() {is=title K2Node_FunctionEntry}
   > n1 = For Each Loop(Array:Array of Audio Component Object References=Get CurrentlyPlayingSounds()) -> (Array Element:Audio Component Object Reference)
   > n2 = Set CurrentAudioComponent(CurrentAudioComponent=@n1)
   > Remove Item(TargetArray:Array of Audio Component Object References=Get CurrentlyPlayingSounds(), Item:"Audio Component Object Reference (by ref)"=@n2)
   > n3 = Get()
   > Stop(self:Audio Component Object Reference=@n3)
  ]}
 TransactionAdd {nodes [
   n1 = TransactionAdd!() -> (then, Object:Object Reference) {is=title K2Node_FunctionEntry}
   > TransactObject(Object=@n1)
  ]}
 TransactionEnd {nodes [
   TransactionEnd() {is=title K2Node_FunctionEntry}
   > EndTransaction()
  ]}
 TransactionStart {nodes [
   n1 = TransactionStart!() -> (then, Description:Text) {is=title K2Node_FunctionEntry}
   n2 = Self-Reference()
   > BeginTransaction(Context=SoundLibraryEditor, Description=@n1, PrimaryObject=@n2)
   > TransactObject(Object=@n2)
  ]}
]
variables [
 AllEntryObjects {type="array<SoundLibraryEditor_LibraryEntry_C>", default []}
 bIsEntryHovered {type=boolean}
 CachedLibrary {type=SoundLibrary, default=null}
 CreateAssetScriptSoundCue {
  type=string
  default="import unreal\r\n\r\nfactory = unreal.SoundCueFactoryNew()\r\n\r\nasset = unreal.AssetToolsHelpers.get_asset_tools().create_asset(name, path, unreal.SoundCue, factory)"
 }
 CreateAssetScriptSoundCueTemplate {
  type=string
  default="import unreal\r\n\r\nfactory = unreal.SoundCueTemplateFactory()\r\nfactory.set_editor_property(\"SoundCueTemplateClass\", templateclass)\r\n\r\nasset = unreal.AssetToolsHelpers.get_asset_tools().create_asset(name, path, unreal.SoundCueTemplate, factory)"
 }
 CurrentActionObject {type=Object, default=null}
 CurrentActionType {type=SoundLibraryEditor_ActionTypes, default=Add Tags}
 CurrentlyPlayingSounds {type="array<AudioComponent>", default []}
 LastSelectedWidget {type=Object, default=null}
 Library {type=SoundLibrary, default=null}
 PlaybackDistance {type=number}
 SelectedEntries {type="array<Object>", default []}
 ShiftStartSelection {type=EUW_SoundLibraryEditor_LibraryEntry_C, default=null}
 TagsToIgnore {type="array<string>", default []}
]
widgets {
 info {rootWidgetClass=&CanvasPanel, widgetCount=55}
 widgets [
  CanvasPanel_0 {widgetClassPath=&CanvasPanel, widget=CanvasPanel_0}
  SizeBox_0 {
   widgetClassPath=&SizeBox
   parent=CanvasPanel_0
   widget=SizeBox_0
   slot {
    class=CanvasPanelSlot
    properties {bAutoSize=true, layoutData {alignment=<0 0>, anchors {maximum=<1 1>, minimum=<0 0>}, offsets=<ltrb 0 0 0 0>}}
   }
  }
  HorizontalBox_180 {widgetClassPath=&HorizontalBox, parent=SizeBox_0, widget=HorizontalBox_180, slot {class=SizeBoxSlot}}
  VerticalBox_38 {
   widgetClassPath=&VerticalBox
   parent=HorizontalBox_180
   widget=VerticalBox_38
   slot {class=HorizontalBoxSlot, properties {size {sizeRule=Fill, value=0.75}}}
  }
  Border_154 {
   widgetClassPath=&Border
   parent=VerticalBox_38
   widget=Border_154
   properties {
    background {
     bIsDynamicallyLoaded=false
     drawAs=Box
     imageType=NoImage
     mirroring=NoMirror
     resourceName=None
     resourceObject=null
     tiling=NoTile
     imageSize=<32 32>
     margin=<ltrb 0 0 0 0>
     outlineSettings {
      bUseBrushTransparency=false
      roundingType=HalfHeightRadius
      width=0
      color {colorUseRule=UseColor_Specified, specifiedColor=<rgba 0 0 0 0>}
      cornerRadii=<0 0 0 0>
     }
     tintColor {colorUseRule=UseColor_Specified, specifiedColor=<rgba 1 1 1 1>}
     uVRegion {bIsValid=false, max=<0 0>, min=<0 0>}
    }
    brushColor=<rgba 0 0 0 0.2>
    padding=<ltrb 0 0 0 0>
   }
   slot {class=VerticalBoxSlot, properties {padding=<ltrb 2.5 2.5 2.5 2.5>, size {sizeRule=Fill, value=1}}}
  }
  HorizontalBox_233 {
   widgetClassPath=&HorizontalBox
   parent=Border_154
   widget=HorizontalBox_233
   slot {class=BorderSlot, properties {padding=<ltrb 0 0 0 0>}}
  }
  VerticalBox_1 {
   widgetClassPath=&VerticalBox
   parent=HorizontalBox_233
   widget=VerticalBox_1
   slot {class=HorizontalBoxSlot, properties {padding=<ltrb 5 5 5 5>, size {sizeRule=Fill, value=1}}}
  }
  Border {
   widgetClassPath=&Border
   parent=VerticalBox_1
   widget=Border
   properties {
    background {
     bIsDynamicallyLoaded=false
     drawAs=Box
     imageType=NoImage
     mirroring=NoMirror
     resourceName=None
     resourceObject=null
     tiling=NoTile
     imageSize=<32 32>
     margin=<ltrb 0 0 0 0>
     outlineSettings {
      bUseBrushTransparency=false
      roundingType=HalfHeightRadius
      width=0
      color {colorUseRule=UseColor_Specified, specifiedColor=<rgba 0 0 0 0>}
      cornerRadii=<0 0 0 0>
     }
     tintColor {colorUseRule=UseColor_Specified, specifiedColor=<rgba 1 1 1 1>}
     uVRegion {bIsValid=false, max=<0 0>, min=<0 0>}
    }
    brushColor=<rgba 0 0 0 0.2>
    padding=<ltrb 0 0 0 0>
   }
   slot {class=VerticalBoxSlot, properties {verticalAlignment=VAlign_Center, size {sizeRule=Automatic, value=0.05}}}
  }
  VerticalBox_210 {widgetClassPath=&VerticalBox, parent=Border, widget=VerticalBox_210, slot {class=BorderSlot, properties {padding=<ltrb 0 0 0 0>}}}
  HorizontalBox_283 {widgetClassPath=&HorizontalBox, parent=VerticalBox_210, widget=HorizontalBox_283, slot {class=VerticalBoxSlot}}
  SaveButton {
   widgetClassPath=&Button
   parent=HorizontalBox_283
   widget=SaveButton
   properties {widgetStyle {
     clickedSlateSound {resourceObject=null}
     disabled {
      bIsDynamicallyLoaded=false
      drawAs=Box
      imageType=NoImage
      mirroring=NoMirror
      resourceName=None
      resourceObject=null
      tiling=NoTile
      imageSize=<0 0>
      margin=<ltrb 0 0 0 0>
      outlineSettings {
       bUseBrushTransparency=false
       roundingType=FixedRadius
       width=1
       color {colorUseRule=UseColor_ColorTable, specifiedColor=<rgba 0 0 0 0>}
       cornerRadii=<4 4 4 4>
      }
      tintColor {colorUseRule=UseColor_Specified, specifiedColor=<rgba 0.0395 0.0395 0.0395 1>}
      uVRegion {bIsValid=false, max=<0 0>, min=<0 0>}
     }
     disabledForeground {colorUseRule=UseColor_ColorTable, specifiedColor=<rgba 0 0 0 0>}
     hovered {
      bIsDynamicallyLoaded=false
      drawAs=Box
      imageType=NoImage
      mirroring=NoMirror
      resourceName=None
      resourceObject=null
      tiling=NoTile
      imageSize=<32 32>
      margin=<ltrb 0 0 0 0>
      outlineSettings {
       bUseBrushTransparency=false
       roundingType=FixedRadius
       width=1
       color {colorUseRule=UseColor_ColorTable, specifiedColor=<rgba 0 0 0 0>}
       cornerRadii=<4 4 4 4>
      }
      tintColor {colorUseRule=UseColor_Specified, specifiedColor=<rgba 0.0953 0.0953 0.0953 1>}
      uVRegion {bIsValid=false, max=<0 0>, min=<0 0>}
     }
     hoveredForeground {colorUseRule=UseColor_ColorTable, specifiedColor=<rgba 0 0 0 0>}
     hoveredSlateSound {resourceObject=null}
     normal {
      bIsDynamicallyLoaded=false
      drawAs=Box
      imageType=NoImage
      mirroring=NoMirror
      resourceName=None
      resourceObject=null
      tiling=NoTile
      imageSize=<32 32>
      margin=<ltrb 0 0 0 0>
      outlineSettings {
       bUseBrushTransparency=false
       roundingType=FixedRadius
       width=1
       color {colorUseRule=UseColor_ColorTable, specifiedColor=<rgba 0 0 0 0>}
       cornerRadii=<4 4 4 4>
      }
      tintColor {colorUseRule=UseColor_Specified, specifiedColor=<rgba 0.0284 0.0284 0.0284 1>}
      uVRegion {bIsValid=false, max=<0 0>, min=<0 0>}
     }
     normalForeground {colorUseRule=UseColor_ColorTable, specifiedColor=<rgba 0 0 0 0>}
     normalPadding=<ltrb 12 1.5 12 1.5>
     pressed {
      bIsDynamicallyLoaded=false
      drawAs=Box
      imageType=NoImage
      mirroring=NoMirror
      resourceName=None
      resourceObject=null
      tiling=NoTile
      imageSize=<32 32>
      margin=<ltrb 0 0 0 0>
      outlineSettings {
       bUseBrushTransparency=false
       roundingType=FixedRadius
       width=1
       color {colorUseRule=UseColor_ColorTable, specifiedColor=<rgba 0 0 0 0>}
       cornerRadii=<4 4 4 4>
      }
      tintColor {colorUseRule=UseColor_Specified, specifiedColor=<rgba 0.0284 0.0284 0.0284 1>}
      uVRegion {bIsValid=false, max=<0 0>, min=<0 0>}
     }
     pressedForeground {colorUseRule=UseColor_ColorTable, specifiedColor=<rgba 0 0 0 0>}
     pressedPadding=<ltrb 12 2.5 12 0.5>
     pressedSlateSound {resourceObject=null}
    }}
   slot {class=HorizontalBoxSlot, properties {padding=<ltrb 5 5 5 5>, size {sizeRule=Fill, value=1}}}
  }
  SaveText {
   widgetClassPath=&UEFN_TextBlock
   bIsVariable=true
   parent=SaveButton
   widget=SaveText
   properties {
    clipping=ClipToBounds
    text=Save
    font {
     bForceMonospaced=false
     bMaterialIsStencil=false
     fontMaterial=null
     fontObject=&NotoSans
     letterSpacing=5
     monospacedWidth=1
     size=10
     skewAmount=0
     typefaceFontName=NotoSans-Regular
     outlineSettings {
      bApplyOutlineToDropShadows=false
      bMiteredCorners=false
      bSeparateFillAlpha=false
      outlineMaterial=null
      outlineSize=0
      outlineColor=<rgba 0 0 0 1>
     }
    }
   }
   slot {class=ButtonSlot, properties {padding=<ltrb 0 0 0 0>}}
  }
  HorizontalBox_108 {
   widgetClassPath=&HorizontalBox
   parent=VerticalBox_210
   widget=HorizontalBox_108
   slot {class=VerticalBoxSlot, properties {verticalAlignment=VAlign_Center, padding=<ltrb 5 5 5 5>}}
  }
  AddEntryButton {
   widgetClassPath=&Button
   parent=HorizontalBox_108
   widget=AddEntryButton
   properties {widgetStyle {
     clickedSlateSound {resourceObject=null}
     disabled {
      bIsDynamicallyLoaded=false
      drawAs=Box
      imageType=NoImage
      mirroring=NoMirror
      resourceName=None
      resourceObject=null
      tiling=NoTile
      imageSize=<0 0>
      margin=<ltrb 0 0 0 0>
      outlineSettings {
       bUseBrushTransparency=false
       roundingType=FixedRadius
       width=1
       color {colorUseRule=UseColor_ColorTable, specifiedColor=<rgba 0 0 0 0>}
       cornerRadii=<4 4 4 4>
      }
      tintColor {colorUseRule=UseColor_Specified, specifiedColor=<rgba 0.0395 0.0395 0.0395 1>}
      uVRegion {bIsValid=false, max=<0 0>, min=<0 0>}
     }
     disabledForeground {colorUseRule=UseColor_ColorTable, specifiedColor=<rgba 0 0 0 0>}
     hovered {
      bIsDynamicallyLoaded=false
      drawAs=Box
      imageType=NoImage
      mirroring=NoMirror
      resourceName=None
      resourceObject=null
      tiling=NoTile
      imageSize=<32 32>
      margin=<ltrb 0 0 0 0>
      outlineSettings {
       bUseBrushTransparency=false
       roundingType=FixedRadius
       width=1
       color {colorUseRule=UseColor_ColorTable, specifiedColor=<rgba 0 0 0 0>}
       cornerRadii=<4 4 4 4>
      }
      tintColor {colorUseRule=UseColor_Specified, specifiedColor=<rgba 0.0953 0.0953 0.0953 1>}
      uVRegion {bIsValid=false, max=<0 0>, min=<0 0>}
     }
     hoveredForeground {colorUseRule=UseColor_ColorTable, specifiedColor=<rgba 0 0 0 0>}
     hoveredSlateSound {resourceObject=null}
     normal {
      bIsDynamicallyLoaded=false
      drawAs=Box
      imageType=NoImage
      mirroring=NoMirror
      resourceName=None
      resourceObject=null
      tiling=NoTile
      imageSize=<32 32>
      margin=<ltrb 0 0 0 0>
      outlineSettings {
       bUseBrushTransparency=false
       roundingType=FixedRadius
       width=1
       color {colorUseRule=UseColor_ColorTable, specifiedColor=<rgba 0 0 0 0>}
       cornerRadii=<4 4 4 4>
      }
      tintColor {colorUseRule=UseColor_Specified, specifiedColor=<rgba 0.0284 0.0284 0.0284 1>}
      uVRegion {bIsValid=false, max=<0 0>, min=<0 0>}
     }
     normalForeground {colorUseRule=UseColor_ColorTable, specifiedColor=<rgba 0 0 0 0>}
     normalPadding=<ltrb 12 1.5 12 1.5>
     pressed {
      bIsDynamicallyLoaded=false
      drawAs=Box
      imageType=NoImage
      mirroring=NoMirror
      resourceName=None
      resourceObject=null
      tiling=NoTile
      imageSize=<32 32>
      margin=<ltrb 0 0 0 0>
      outlineSettings {
       bUseBrushTransparency=false
       roundingType=FixedRadius
       width=1
       color {colorUseRule=UseColor_ColorTable, specifiedColor=<rgba 0 0 0 0>}
       cornerRadii=<4 4 4 4>
      }
      tintColor {colorUseRule=UseColor_Specified, specifiedColor=<rgba 0.0284 0.0284 0.0284 1>}
      uVRegion {bIsValid=false, max=<0 0>, min=<0 0>}
     }
     pressedForeground {colorUseRule=UseColor_ColorTable, specifiedColor=<rgba 0 0 0 0>}
     pressedPadding=<ltrb 12 2.5 12 0.5>
     pressedSlateSound {resourceObject=null}
    }}
   slot {class=HorizontalBoxSlot, properties {padding=<ltrb 0 0 2.5 0>, size {sizeRule=Fill, value=1}}}
  }
  Text_4 {
   widgetClassPath=&UEFN_TextBlock
   bIsVariable=true
   parent=AddEntryButton
   widget=Text_4
   properties {
    clipping=ClipToBounds
    text=Add
    font {
     bForceMonospaced=false
     bMaterialIsStencil=false
     fontMaterial=null
     fontObject=&NotoSans
     letterSpacing=5
     monospacedWidth=1
     size=10
     skewAmount=0
     typefaceFontName=NotoSans-Regular
     outlineSettings {
      bApplyOutlineToDropShadows=false
      bMiteredCorners=false
      bSeparateFillAlpha=false
      outlineMaterial=null
      outlineSize=0
      outlineColor=<rgba 0 0 0 1>
     }
    }
   }
   slot {class=ButtonSlot, properties {padding=<ltrb 0 0 0 0>}}
  }
  RemoveEntry {
   widgetClassPath=&Button
   parent=HorizontalBox_108
   widget=RemoveEntry
   properties {widgetStyle {
     clickedSlateSound {resourceObject=null}
     disabled {
      bIsDynamicallyLoaded=false
      drawAs=Box
      imageType=NoImage
      mirroring=NoMirror
      resourceName=None
      resourceObject=null
      tiling=NoTile
      imageSize=<0 0>
      margin=<ltrb 0 0 0 0>
      outlineSettings {
       bUseBrushTransparency=false
       roundingType=FixedRadius
       width=1
       color {colorUseRule=UseColor_ColorTable, specifiedColor=<rgba 0 0 0 0>}
       cornerRadii=<4 4 4 4>
      }
      tintColor {colorUseRule=UseColor_Specified, specifiedColor=<rgba 0.0395 0.0395 0.0395 1>}
      uVRegion {bIsValid=false, max=<0 0>, min=<0 0>}
     }
     disabledForeground {colorUseRule=UseColor_ColorTable, specifiedColor=<rgba 0 0 0 0>}
     hovered {
      bIsDynamicallyLoaded=false
      drawAs=Box
      imageType=NoImage
      mirroring=NoMirror
      resourceName=None
      resourceObject=null
      tiling=NoTile
      imageSize=<32 32>
      margin=<ltrb 0 0 0 0>
      outlineSettings {
       bUseBrushTransparency=false
       roundingType=FixedRadius
       width=1
       color {colorUseRule=UseColor_ColorTable, specifiedColor=<rgba 0 0 0 0>}
       cornerRadii=<4 4 4 4>
      }
      tintColor {colorUseRule=UseColor_Specified, specifiedColor=<rgba 0.0953 0.0953 0.0953 1>}
      uVRegion {bIsValid=false, max=<0 0>, min=<0 0>}
     }
     hoveredForeground {colorUseRule=UseColor_ColorTable, specifiedColor=<rgba 0 0 0 0>}
     hoveredSlateSound {resourceObject=null}
     normal {
      bIsDynamicallyLoaded=false
      drawAs=Box
      imageType=NoImage
      mirroring=NoMirror
      resourceName=None
      resourceObject=null
      tiling=NoTile
      imageSize=<32 32>
      margin=<ltrb 0 0 0 0>
      outlineSettings {
       bUseBrushTransparency=false
       roundingType=FixedRadius
       width=1
       color {colorUseRule=UseColor_ColorTable, specifiedColor=<rgba 0 0 0 0>}
       cornerRadii=<4 4 4 4>
      }
      tintColor {colorUseRule=UseColor_Specified, specifiedColor=<rgba 0.0284 0.0284 0.0284 1>}
      uVRegion {bIsValid=false, max=<0 0>, min=<0 0>}
     }
     normalForeground {colorUseRule=UseColor_ColorTable, specifiedColor=<rgba 0 0 0 0>}
     normalPadding=<ltrb 12 1.5 12 1.5>
     pressed {
      bIsDynamicallyLoaded=false
      drawAs=Box
      imageType=NoImage
      mirroring=NoMirror
      resourceName=None
      resourceObject=null
      tiling=NoTile
      imageSize=<32 32>
      margin=<ltrb 0 0 0 0>
      outlineSettings {
       bUseBrushTransparency=false
       roundingType=FixedRadius
       width=1
       color {colorUseRule=UseColor_ColorTable, specifiedColor=<rgba 0 0 0 0>}
       cornerRadii=<4 4 4 4>
      }
      tintColor {colorUseRule=UseColor_Specified, specifiedColor=<rgba 0.0284 0.0284 0.0284 1>}
      uVRegion {bIsValid=false, max=<0 0>, min=<0 0>}
     }
     pressedForeground {colorUseRule=UseColor_ColorTable, specifiedColor=<rgba 0 0 0 0>}
     pressedPadding=<ltrb 12 2.5 12 0.5>
     pressedSlateSound {resourceObject=null}
    }}
   slot {class=HorizontalBoxSlot, properties {padding=<ltrb 2.5 0 0 0>, size {sizeRule=Fill, value=1}}}
  }
  Text_2 {
   widgetClassPath=&UEFN_TextBlock
   bIsVariable=true
   parent=RemoveEntry
   widget=Text_2
   properties {
    clipping=ClipToBounds
    text=Remove
    font {
     bForceMonospaced=false
     bMaterialIsStencil=false
     fontMaterial=null
     fontObject=&NotoSans
     letterSpacing=5
     monospacedWidth=1
     size=10
     skewAmount=0
     typefaceFontName=NotoSans-Regular
     outlineSettings {
      bApplyOutlineToDropShadows=false
      bMiteredCorners=false
      bSeparateFillAlpha=false
      outlineMaterial=null
      outlineSize=0
      outlineColor=<rgba 0 0 0 1>
     }
    }
   }
   slot {class=ButtonSlot, properties {padding=<ltrb 0 0 0 0>}}
  }
  EntryListScrollBox {
   widgetClassPath=&ScrollBox
   parent=VerticalBox_1
   widget=EntryListScrollBox
   properties {
    widgetBarStyle {
     thickness=8
     draggedThumbImage {
      bIsDynamicallyLoaded=false
      drawAs=RoundedBox
      imageType=NoImage
      mirroring=NoMirror
      resourceName=None
      resourceObject=null
      tiling=NoTile
      imageSize=<0 0>
      margin=<ltrb 0 0 0 0>
      outlineSettings {
       bUseBrushTransparency=false
       roundingType=FixedRadius
       width=0
       color {colorUseRule=UseColor_Specified, specifiedColor=<rgba 0 0 0 0>}
       cornerRadii=<4 4 4 4>
      }
      tintColor {colorUseRule=UseColor_Specified, specifiedColor=<rgba 0.2159 0.2159 0.2159 1>}
      uVRegion {bIsValid=false, max=<0 0>, min=<0 0>}
     }
     horizontalBackgroundImage {
      bIsDynamicallyLoaded=false
      drawAs=NoDrawType
      imageType=NoImage
      mirroring=NoMirror
      resourceName=None
      resourceObject=null
      tiling=NoTile
      imageSize=<0 0>
      margin=<ltrb 0 0 0 0>
      outlineSettings {
       bUseBrushTransparency=false
       roundingType=HalfHeightRadius
       width=0
       color {colorUseRule=UseColor_Specified, specifiedColor=<rgba 0 0 0 0>}
       cornerRadii=<0 0 0 0>
      }
      tintColor {colorUseRule=UseColor_Specified, specifiedColor=<rgba 1 1 1 1>}
      uVRegion {bIsValid=false, max=<0 0>, min=<0 0>}
     }
     horizontalBottomSlotImage {
      bIsDynamicallyLoaded=false
      drawAs=NoDrawType
      imageType=NoImage
      mirroring=NoMirror
      resourceName=None
      resourceObject=null
      tiling=NoTile
      imageSize=<0 0>
      margin=<ltrb 0 0 0 0>
      outlineSettings {
       bUseBrushTransparency=false
       roundingType=HalfHeightRadius
       width=0
       color {colorUseRule=UseColor_Specified, specifiedColor=<rgba 0 0 0 0>}
       cornerRadii=<0 0 0 0>
      }
      tintColor {colorUseRule=UseColor_Specified, specifiedColor=<rgba 1 1 1 1>}
      uVRegion {bIsValid=false, max=<0 0>, min=<0 0>}
     }
     horizontalTopSlotImage {
      bIsDynamicallyLoaded=false
      drawAs=NoDrawType
      imageType=NoImage
      mirroring=NoMirror
      resourceName=None
      resourceObject=null
      tiling=NoTile
      imageSize=<0 0>
      margin=<ltrb 0 0 0 0>
      outlineSettings {
       bUseBrushTransparency=false
       roundingType=HalfHeightRadius
       width=0
       color {colorUseRule=UseColor_Specified, specifiedColor=<rgba 0 0 0 0>}
       cornerRadii=<0 0 0 0>
      }
      tintColor {colorUseRule=UseColor_Specified, specifiedColor=<rgba 1 1 1 1>}
      uVRegion {bIsValid=false, max=<0 0>, min=<0 0>}
     }
     hoveredThumbImage {
      bIsDynamicallyLoaded=false
      drawAs=RoundedBox
      imageType=NoImage
      mirroring=NoMirror
      resourceName=None
      resourceObject=null
      tiling=NoTile
      imageSize=<0 0>
      margin=<ltrb 0 0 0 0>
      outlineSettings {
       bUseBrushTransparency=false
       roundingType=FixedRadius
       width=0
       color {colorUseRule=UseColor_Specified, specifiedColor=<rgba 0 0 0 0>}
       cornerRadii=<4 4 4 4>
      }
      tintColor {colorUseRule=UseColor_Specified, specifiedColor=<rgba 0.2159 0.2159 0.2159 1>}
      uVRegion {bIsValid=false, max=<0 0>, min=<0 0>}
     }
     normalThumbImage {
      bIsDynamicallyLoaded=false
      drawAs=RoundedBox
      imageType=NoImage
      mirroring=NoMirror
      resourceName=None
      resourceObject=null
      tiling=NoTile
      imageSize=<0 0>
      margin=<ltrb 0 0 0 0>
      outlineSettings {
       bUseBrushTransparency=false
       roundingType=FixedRadius
       width=0
       color {colorUseRule=UseColor_Specified, specifiedColor=<rgba 0 0 0 0>}
       cornerRadii=<4 4 4 4>
      }
      tintColor {colorUseRule=UseColor_Specified, specifiedColor=<rgba 0.0953 0.0953 0.0953 1>}
      uVRegion {bIsValid=false, max=<0 0>, min=<0 0>}
     }
     verticalBackgroundImage {
      bIsDynamicallyLoaded=false
      drawAs=NoDrawType
      imageType=NoImage
      mirroring=NoMirror
      resourceName=None
      resourceObject=null
      tiling=NoTile
      imageSize=<0 0>
      margin=<ltrb 0 0 0 0>
      outlineSettings {
       bUseBrushTransparency=false
       roundingType=HalfHeightRadius
       width=0
       color {colorUseRule=UseColor_Specified, specifiedColor=<rgba 0 0 0 0>}
       cornerRadii=<0 0 0 0>
      }
      tintColor {colorUseRule=UseColor_Specified, specifiedColor=<rgba 1 1 1 1>}
      uVRegion {bIsValid=false, max=<0 0>, min=<0 0>}
     }
     verticalBottomSlotImage {
      bIsDynamicallyLoaded=false
      drawAs=NoDrawType
      imageType=NoImage
      mirroring=NoMirror
      resourceName=None
      resourceObject=null
      tiling=NoTile
      imageSize=<0 0>
      margin=<ltrb 0 0 0 0>
      outlineSettings {
       bUseBrushTransparency=false
       roundingType=HalfHeightRadius
       width=0
       color {colorUseRule=UseColor_Specified, specifiedColor=<rgba 0 0 0 0>}
       cornerRadii=<0 0 0 0>
      }
      tintColor {colorUseRule=UseColor_Specified, specifiedColor=<rgba 1 1 1 1>}
      uVRegion {bIsValid=false, max=<0 0>, min=<0 0>}
     }
     verticalTopSlotImage {
      bIsDynamicallyLoaded=false
      drawAs=NoDrawType
      imageType=NoImage
      mirroring=NoMirror
      resourceName=None
      resourceObject=null
      tiling=NoTile
      imageSize=<0 0>
      margin=<ltrb 0 0 0 0>
      outlineSettings {
       bUseBrushTransparency=false
       roundingType=HalfHeightRadius
       width=0
       color {colorUseRule=UseColor_Specified, specifiedColor=<rgba 0 0 0 0>}
       cornerRadii=<0 0 0 0>
      }
      tintColor {colorUseRule=UseColor_Specified, specifiedColor=<rgba 1 1 1 1>}
      uVRegion {bIsValid=false, max=<0 0>, min=<0 0>}
     }
    }
    widgetStyle {
     barThickness=8
     bottomShadowBrush {
      bIsDynamicallyLoaded=false
      drawAs=Box
      imageType=FullColor
      mirroring=NoMirror
      resourceName=../../../Engine/Content/Slate/Common/ScrollBoxShadowBottom.png
      resourceObject=null
      tiling=NoTile
      imageSize=<16 8>
      margin=<ltrb 0.5 0 0.5 1>
      outlineSettings {
       bUseBrushTransparency=false
       roundingType=HalfHeightRadius
       width=0
       color {colorUseRule=UseColor_Specified, specifiedColor=<rgba 0 0 0 0>}
       cornerRadii=<0 0 0 0>
      }
      tintColor {colorUseRule=UseColor_Specified, specifiedColor=<rgba 1 1 1 1>}
      uVRegion {bIsValid=false, max=<0 0>, min=<0 0>}
     }
     horizontalScrolledContentPadding=<ltrb 0 0 1 0>
     leftShadowBrush {
      bIsDynamicallyLoaded=false
      drawAs=Box
      imageType=FullColor
      mirroring=NoMirror
      resourceName=../../../Engine/Content/Slate/Common/ScrollBoxShadowLeft.png
      resourceObject=null
      tiling=NoTile
      imageSize=<8 16>
      margin=<ltrb 1 0.5 0 0.5>
      outlineSettings {
       bUseBrushTransparency=false
       roundingType=HalfHeightRadius
       width=0
       color {colorUseRule=UseColor_Specified, specifiedColor=<rgba 0 0 0 0>}
       cornerRadii=<0 0 0 0>
      }
      tintColor {colorUseRule=UseColor_Specified, specifiedColor=<rgba 1 1 1 1>}
      uVRegion {bIsValid=false, max=<0 0>, min=<0 0>}
     }
     rightShadowBrush {
      bIsDynamicallyLoaded=false
      drawAs=Box
      imageType=FullColor
      mirroring=NoMirror
      resourceName=../../../Engine/Content/Slate/Common/ScrollBoxShadowRight.png
      resourceObject=null
      tiling=NoTile
      imageSize=<8 16>
      margin=<ltrb 0 0.5 1 0.5>
      outlineSettings {
       bUseBrushTransparency=false
       roundingType=HalfHeightRadius
       width=0
       color {colorUseRule=UseColor_Specified, specifiedColor=<rgba 0 0 0 0>}
       cornerRadii=<0 0 0 0>
      }
      tintColor {colorUseRule=UseColor_Specified, specifiedColor=<rgba 1 1 1 1>}
      uVRegion {bIsValid=false, max=<0 0>, min=<0 0>}
     }
     topShadowBrush {
      bIsDynamicallyLoaded=false
      drawAs=Box
      imageType=FullColor
      mirroring=NoMirror
      resourceName=../../../Engine/Content/Slate/Common/ScrollBoxShadowTop.png
      resourceObject=null
      tiling=NoTile
      imageSize=<16 8>
      margin=<ltrb 0.5 1 0.5 0>
      outlineSettings {
       bUseBrushTransparency=false
       roundingType=HalfHeightRadius
       width=0
       color {colorUseRule=UseColor_Specified, specifiedColor=<rgba 0 0 0 0>}
       cornerRadii=<0 0 0 0>
      }
      tintColor {colorUseRule=UseColor_Specified, specifiedColor=<rgba 1 1 1 1>}
      uVRegion {bIsValid=false, max=<0 0>, min=<0 0>}
     }
     verticalScrolledContentPadding=<ltrb 0 0 0 1>
    }
   }
   slot {class=VerticalBoxSlot, properties {size {sizeRule=Fill, value=1}}}
  }
  SoundLibEntries {
   widgetClassPath=/Script/UMG.ListView
   bIsVariable=true
   parent=EntryListScrollBox
   widget=SoundLibEntries
   properties {
    entryWidgetClass=&/EUW_SoundLibraryEditor_LibraryEntry.EUW_SoundLibraryEditor_LibraryEntry_C
    selectionMode=None
    scrollBarStyle {
     thickness=8
     draggedThumbImage {
      bIsDynamicallyLoaded=false
      drawAs=RoundedBox
      imageType=NoImage
      mirroring=NoMirror
      resourceName=None
      resourceObject=null
      tiling=NoTile
      imageSize=<0 0>
      margin=<ltrb 0 0 0 0>
      outlineSettings {
       bUseBrushTransparency=false
       roundingType=FixedRadius
       width=0
       color {colorUseRule=UseColor_Specified, specifiedColor=<rgba 0 0 0 0>}
       cornerRadii=<4 4 4 4>
      }
      tintColor {colorUseRule=UseColor_Specified, specifiedColor=<rgba 0.2159 0.2159 0.2159 1>}
      uVRegion {bIsValid=false, max=<0 0>, min=<0 0>}
     }
     horizontalBackgroundImage {
      bIsDynamicallyLoaded=false
      drawAs=NoDrawType
      imageType=NoImage
      mirroring=NoMirror
      resourceName=None
      resourceObject=null
      tiling=NoTile
      imageSize=<0 0>
      margin=<ltrb 0 0 0 0>
      outlineSettings {
       bUseBrushTransparency=false
       roundingType=HalfHeightRadius
       width=0
       color {colorUseRule=UseColor_Specified, specifiedColor=<rgba 0 0 0 0>}
       cornerRadii=<0 0 0 0>
      }
      tintColor {colorUseRule=UseColor_Specified, specifiedColor=<rgba 1 1 1 1>}
      uVRegion {bIsValid=false, max=<0 0>, min=<0 0>}
     }
     horizontalBottomSlotImage {
      bIsDynamicallyLoaded=false
      drawAs=NoDrawType
      imageType=NoImage
      mirroring=NoMirror
      resourceName=None
      resourceObject=null
      tiling=NoTile
      imageSize=<0 0>
      margin=<ltrb 0 0 0 0>
      outlineSettings {
       bUseBrushTransparency=false
       roundingType=HalfHeightRadius
       width=0
       color {colorUseRule=UseColor_Specified, specifiedColor=<rgba 0 0 0 0>}
       cornerRadii=<0 0 0 0>
      }
      tintColor {colorUseRule=UseColor_Specified, specifiedColor=<rgba 1 1 1 1>}
      uVRegion {bIsValid=false, max=<0 0>, min=<0 0>}
     }
     horizontalTopSlotImage {
      bIsDynamicallyLoaded=false
      drawAs=NoDrawType
      imageType=NoImage
      mirroring=NoMirror
      resourceName=None
      resourceObject=null
      tiling=NoTile
      imageSize=<0 0>
      margin=<ltrb 0 0 0 0>
      outlineSettings {
       bUseBrushTransparency=false
       roundingType=HalfHeightRadius
       width=0
       color {colorUseRule=UseColor_Specified, specifiedColor=<rgba 0 0 0 0>}
       cornerRadii=<0 0 0 0>
      }
      tintColor {colorUseRule=UseColor_Specified, specifiedColor=<rgba 1 1 1 1>}
      uVRegion {bIsValid=false, max=<0 0>, min=<0 0>}
     }
     hoveredThumbImage {
      bIsDynamicallyLoaded=false
      drawAs=RoundedBox
      imageType=NoImage
      mirroring=NoMirror
      resourceName=None
      resourceObject=null
      tiling=NoTile
      imageSize=<0 0>
      margin=<ltrb 0 0 0 0>
      outlineSettings {
       bUseBrushTransparency=false
       roundingType=FixedRadius
       width=0
       color {colorUseRule=UseColor_Specified, specifiedColor=<rgba 0 0 0 0>}
       cornerRadii=<4 4 4 4>
      }
      tintColor {colorUseRule=UseColor_Specified, specifiedColor=<rgba 0.2159 0.2159 0.2159 1>}
      uVRegion {bIsValid=false, max=<0 0>, min=<0 0>}
     }
     normalThumbImage {
      bIsDynamicallyLoaded=false
      drawAs=RoundedBox
      imageType=NoImage
      mirroring=NoMirror
      resourceName=None
      resourceObject=null
      tiling=NoTile
      imageSize=<0 0>
      margin=<ltrb 0 0 0 0>
      outlineSettings {
       bUseBrushTransparency=false
       roundingType=FixedRadius
       width=0
       color {colorUseRule=UseColor_Specified, specifiedColor=<rgba 0 0 0 0>}
       cornerRadii=<4 4 4 4>
      }
      tintColor {colorUseRule=UseColor_Specified, specifiedColor=<rgba 0.0953 0.0953 0.0953 1>}
      uVRegion {bIsValid=false, max=<0 0>, min=<0 0>}
     }
     verticalBackgroundImage {
      bIsDynamicallyLoaded=false
      drawAs=NoDrawType
      imageType=NoImage
      mirroring=NoMirror
      resourceName=None
      resourceObject=null
      tiling=NoTile
      imageSize=<0 0>
      margin=<ltrb 0 0 0 0>
      outlineSettings {
       bUseBrushTransparency=false
       roundingType=HalfHeightRadius
       width=0
       color {colorUseRule=UseColor_Specified, specifiedColor=<rgba 0 0 0 0>}
       cornerRadii=<0 0 0 0>
      }
      tintColor {colorUseRule=UseColor_Specified, specifiedColor=<rgba 1 1 1 1>}
      uVRegion {bIsValid=false, max=<0 0>, min=<0 0>}
     }
     verticalBottomSlotImage {
      bIsDynamicallyLoaded=false
      drawAs=NoDrawType
      imageType=NoImage
      mirroring=NoMirror
      resourceName=None
      resourceObject=null
      tiling=NoTile
      imageSize=<0 0>
      margin=<ltrb 0 0 0 0>
      outlineSettings {
       bUseBrushTransparency=false
       roundingType=HalfHeightRadius
       width=0
       color {colorUseRule=UseColor_Specified, specifiedColor=<rgba 0 0 0 0>}
       cornerRadii=<0 0 0 0>
      }
      tintColor {colorUseRule=UseColor_Specified, specifiedColor=<rgba 1 1 1 1>}
      uVRegion {bIsValid=false, max=<0 0>, min=<0 0>}
     }
     verticalTopSlotImage {
      bIsDynamicallyLoaded=false
      drawAs=NoDrawType
      imageType=NoImage
      mirroring=NoMirror
      resourceName=None
      resourceObject=null
      tiling=NoTile
      imageSize=<0 0>
      margin=<ltrb 0 0 0 0>
      outlineSettings {
       bUseBrushTransparency=false
       roundingType=HalfHeightRadius
       width=0
       color {colorUseRule=UseColor_Specified, specifiedColor=<rgba 0 0 0 0>}
       cornerRadii=<0 0 0 0>
      }
      tintColor {colorUseRule=UseColor_Specified, specifiedColor=<rgba 1 1 1 1>}
      uVRegion {bIsValid=false, max=<0 0>, min=<0 0>}
     }
    }
   }
   slot {class=ScrollBoxSlot}
  }
  Border_1 {
   widgetClassPath=&Border
   parent=HorizontalBox_233
   widget=Border_1
   properties {
    background {
     bIsDynamicallyLoaded=false
     drawAs=Box
     imageType=NoImage
     mirroring=NoMirror
     resourceName=None
     resourceObject=null
     tiling=NoTile
     imageSize=<32 32>
     margin=<ltrb 0 0 0 0>
     outlineSettings {
      bUseBrushTransparency=false
      roundingType=HalfHeightRadius
      width=0
      color {colorUseRule=UseColor_Specified, specifiedColor=<rgba 0 0 0 0>}
      cornerRadii=<0 0 0 0>
     }
     tintColor {colorUseRule=UseColor_Specified, specifiedColor=<rgba 1 1 1 1>}
     uVRegion {bIsValid=false, max=<0 0>, min=<0 0>}
    }
    brushColor=<rgba 0 0 0 0.2>
    padding=<ltrb 0 0 0 0>
   }
   slot {class=HorizontalBoxSlot, properties {padding=<ltrb 5 5 5 5>, size {sizeRule=Fill, value=0.5}}}
  }
  WidgetSwitcher {
   widgetClassPath=&WidgetSwitcher
   bIsVariable=true
   parent=Border_1
   widget=WidgetSwitcher
   slot {class=BorderSlot, properties {padding=<ltrb 0 0 0 0>}}
  }
  VerticalBox_2 {widgetClassPath=&VerticalBox, parent=WidgetSwitcher, widget=VerticalBox_2, slot {class=WidgetSwitcherSlot}}
  Border_2 {
   widgetClassPath=&Border
   parent=VerticalBox_2
   widget=Border_2
   properties {
    background {
     bIsDynamicallyLoaded=false
     drawAs=Box
     imageType=NoImage
     mirroring=NoMirror
     resourceName=None
     resourceObject=null
     tiling=NoTile
     imageSize=<32 32>
     margin=<ltrb 0 0 0 0>
     outlineSettings {
      bUseBrushTransparency=false
      roundingType=HalfHeightRadius
      width=0
      color {colorUseRule=UseColor_Specified, specifiedColor=<rgba 0 0 0 0>}
      cornerRadii=<0 0 0 0>
     }
     tintColor {colorUseRule=UseColor_Specified, specifiedColor=<rgba 1 1 1 1>}
     uVRegion {bIsValid=false, max=<0 0>, min=<0 0>}
    }
    brushColor=<rgba 0 0 0 0.2>
    padding=<ltrb 0 0 0 5>
   }
   slot {class=VerticalBoxSlot}
  }
  LibPropertyView {
   widgetClassPath=&SinglePropertyView
   parent=Border_2
   widget=LibPropertyView
   properties {nameOverride=" ", propertyName=Library}
   slot {class=BorderSlot, properties {padding=<ltrb 0 0 0 5>}}
  }
  Playback {
   widgetClassPath=&VerticalBox
   bIsVariable=true
   parent=VerticalBox_2
   widget=Playback
   properties {visibility=Collapsed}
   slot {class=VerticalBoxSlot}
  }
  HorizontalBox_54 {widgetClassPath=&HorizontalBox, parent=Playback, widget=HorizontalBox_54, slot {class=VerticalBoxSlot}}
  PlaybackCollapseCheckbox {
   widgetClassPath=/Script/UMG.CheckBox
   bIsVariable=true
   parent=HorizontalBox_54
   widget=PlaybackCollapseCheckbox
   properties {
    renderTransform {angle=0, scale=<1 1>, shear=<0 0>, translation=<0 1>}
    renderTransformPivot=<0.45 0.5>
    widgetStyle {
     checkBoxType=CheckBox
     backgroundHoveredImage {
      bIsDynamicallyLoaded=false
      drawAs=RoundedBox
      imageType=NoImage
      mirroring=NoMirror
      resourceName=None
      resourceObject=null
      tiling=NoTile
      imageSize=<18 18>
      margin=<ltrb 0 0 0 0>
      outlineSettings {
       bUseBrushTransparency=false
       roundingType=FixedRadius
       width=1
       color {colorUseRule=UseColor_ColorTable, specifiedColor=<rgba 0 0 0 0>}
       cornerRadii=<3 3 3 3>
      }
      tintColor {colorUseRule=UseColor_Specified, specifiedColor=<rgba 0.0048 0.0048 0.0048 1>}
      uVRegion {bIsValid=false, max=<0 0>, min=<0 0>}
     }
     backgroundImage {
      bIsDynamicallyLoaded=false
      drawAs=RoundedBox
      imageType=NoImage
      mirroring=NoMirror
      resourceName=None
      resourceObject=null
      tiling=NoTile
      imageSize=<18 18>
      margin=<ltrb 0 0 0 0>
      outlineSettings {
       bUseBrushTransparency=false
       roundingType=FixedRadius
       width=1
       color {colorUseRule=UseColor_ColorTable, specifiedColor=<rgba 0 0 0 0>}
       cornerRadii=<3 3 3 3>
      }
      tintColor {colorUseRule=UseColor_Specified, specifiedColor=<rgba 0.0048 0.0048 0.0048 1>}
      uVRegion {bIsValid=false, max=<0 0>, min=<0 0>}
     }
     backgroundPressedImage {
      bIsDynamicallyLoaded=false
      drawAs=RoundedBox
      imageType=NoImage
      mirroring=NoMirror
      resourceName=None
      resourceObject=null
      tiling=NoTile
      imageSize=<18 18>
      margin=<ltrb 0 0 0 0>
      outlineSettings {
       bUseBrushTransparency=false
       roundingType=FixedRadius
       width=1
       color {colorUseRule=UseColor_ColorTable, specifiedColor=<rgba 0 0 0 0>}
       cornerRadii=<3 3 3 3>
      }
      tintColor {colorUseRule=UseColor_Specified, specifiedColor=<rgba 0.0048 0.0048 0.0048 1>}
      uVRegion {bIsValid=false, max=<0 0>, min=<0 0>}
     }
     borderBackgroundColor {colorUseRule=UseColor_Specified, specifiedColor=<rgba 1 1 1 1>}
     checkedForeground {colorUseRule=UseColor_Specified, specifiedColor=<rgba 1 1 1 1>}
     checkedHoveredForeground {colorUseRule=UseColor_Specified, specifiedColor=<rgba 1 1 1 1>}
     checkedHoveredImage {
      bIsDynamicallyLoaded=false
      drawAs=Image
      imageType=Vector
      mirroring=NoMirror
      resourceName=../../../Engine/Content/Slate/Common/CheckBox.png
      resourceObject=/FortUILibrary/Foundation/Buttons/Materials/Arrows/M_Button_CycleArrow_Right_Hovered
      tiling=NoTile
      imageSize=<16 16>
      margin=<ltrb 0 0 0 0>
      outlineSettings {
       bUseBrushTransparency=false
       roundingType=HalfHeightRadius
       width=0
       color {colorUseRule=UseColor_Specified, specifiedColor=<rgba 0 0 0 0>}
       cornerRadii=<0 0 0 0>
      }
      tintColor {colorUseRule=UseColor_Specified, specifiedColor=<rgba 1 1 1 1>}
      uVRegion {bIsValid=false, max=<0 0>, min=<0 0>}
     }
     checkedImage {
      bIsDynamicallyLoaded=false
      drawAs=Image
      imageType=Vector
      mirroring=NoMirror
      resourceName=../../../Engine/Content/Slate/Common/CheckBox.png
      resourceObject=/FortUILibrary/Foundation/Buttons/Materials/Arrows/M_Button_CycleArrow_Right_Base
      tiling=NoTile
      imageSize=<16 16>
      margin=<ltrb 0 0 0 0>
      outlineSettings {
       bUseBrushTransparency=false
       roundingType=HalfHeightRadius
       width=0
       color {colorUseRule=UseColor_Specified, specifiedColor=<rgba 0 0 0 0>}
       cornerRadii=<0 0 0 0>
      }
      tintColor {colorUseRule=UseColor_Specified, specifiedColor=<rgba 0 0.162 0.7454 1>}
      uVRegion {bIsValid=false, max=<0 0>, min=<0 0>}
     }
     checkedPressedForeground {colorUseRule=UseColor_Specified, specifiedColor=<rgba 1 1 1 1>}
     checkedPressedImage {
      bIsDynamicallyLoaded=false
      drawAs=Image
      imageType=Vector
      mirroring=NoMirror
      resourceName=../../../Engine/Content/Slate/Common/CheckBox.png
      resourceObject=/FortUILibrary/Foundation/Buttons/Materials/Arrows/M_Button_CycleArrow_Right_Pressed
      tiling=NoTile
      imageSize=<16 16>
      margin=<ltrb 0 0 0 0>
      outlineSettings {
       bUseBrushTransparency=false
       roundingType=HalfHeightRadius
       width=0
       color {colorUseRule=UseColor_Specified, specifiedColor=<rgba 0 0 0 0>}
       cornerRadii=<0 0 0 0>
      }
      tintColor {colorUseRule=UseColor_Specified, specifiedColor=<rgba 0 0.0802 0.3515 1>}
      uVRegion {bIsValid=false, max=<0 0>, min=<0 0>}
     }
     checkedSlateSound {resourceObject=null}
     foregroundColor {colorUseRule=UseColor_Specified, specifiedColor=<rgba 1 1 1 1>}
     hoveredForeground {colorUseRule=UseColor_Specified, specifiedColor=<rgba 1 1 1 1>}
     hoveredSlateSound {resourceObject=null}
     padding=<ltrb 2 0 0 0>
     pressedForeground {colorUseRule=UseColor_Specified, specifiedColor=<rgba 1 1 1 1>}
     uncheckedHoveredImage {
      bIsDynamicallyLoaded=false
      drawAs=NoDrawType
      imageType=NoImage
      mirroring=NoMirror
      resourceName=None
      resourceObject=&M_Button_CycleArrow_Right_Hovered
      tiling=NoTile
      imageSize=<0 0>
      margin=<ltrb 0 0 0 0>
      outlineSettings {
       bUseBrushTransparency=false
       roundingType=HalfHeightRadius
       width=0
       color {colorUseRule=UseColor_Specified, specifiedColor=<rgba 0 0 0 0>}
       cornerRadii=<0 0 0 0>
      }
      tintColor {colorUseRule=UseColor_Specified, specifiedColor=<rgba 1 1 1 1>}
      uVRegion {bIsValid=false, max=<0 0>, min=<0 0>}
     }
     uncheckedImage {
      bIsDynamicallyLoaded=false
      drawAs=NoDrawType
      imageType=NoImage
      mirroring=NoMirror
      resourceName=None
      resourceObject=&M_Button_CycleArrow_Right_Base
      tiling=NoTile
      imageSize=<0 0>
      margin=<ltrb 0 0 0 0>
      outlineSettings {
       bUseBrushTransparency=false
       roundingType=HalfHeightRadius
       width=0
       color {colorUseRule=UseColor_Specified, specifiedColor=<rgba 0 0 0 0>}
       cornerRadii=<0 0 0 0>
      }
      tintColor {colorUseRule=UseColor_Specified, specifiedColor=<rgba 1 1 1 1>}
      uVRegion {bIsValid=false, max=<0 0>, min=<0 0>}
     }
     uncheckedPressedImage {
      bIsDynamicallyLoaded=false
      drawAs=NoDrawType
      imageType=NoImage
      mirroring=NoMirror
      resourceName=../../../Engine/Content/Slate/Common/CheckBox.png
      resourceObject=&M_Button_CycleArrow_Right_Pressed
      tiling=NoTile
      imageSize=<0 0>
      margin=<ltrb 0 0 0 0>
      outlineSettings {
       bUseBrushTransparency=false
       roundingType=HalfHeightRadius
       width=0
       color {colorUseRule=UseColor_Specified, specifiedColor=<rgba 0 0 0 0>}
       cornerRadii=<0 0 0 0>
      }
      tintColor {colorUseRule=UseColor_Specified, specifiedColor=<rgba 1 1 1 1>}
      uVRegion {bIsValid=false, max=<0 0>, min=<0 0>}
     }
     uncheckedSlateSound {resourceObject=null}
     undeterminedForeground {colorUseRule=UseColor_Specified, specifiedColor=<rgba 1 1 1 1>}
     undeterminedHoveredImage {
      bIsDynamicallyLoaded=false
      drawAs=Image
      imageType=Vector
      mirroring=NoMirror
      resourceName=../../../Engine/Content/Slate/Starship/CoreWidgets/CheckBox/indeterminate.svg
      resourceObject=null
      tiling=NoTile
      imageSize=<16 16>
      margin=<ltrb 0 0 0 0>
      outlineSettings {
       bUseBrushTransparency=false
       roundingType=HalfHeightRadius
       width=0
       color {colorUseRule=UseColor_Specified, specifiedColor=<rgba 0 0 0 0>}
       cornerRadii=<0 0 0 0>
      }
      tintColor {colorUseRule=UseColor_Specified, specifiedColor=<rgba 0.0044 0.2384 1 1>}
      uVRegion {bIsValid=false, max=<0 0>, min=<0 0>}
     }
     undeterminedImage {
      bIsDynamicallyLoaded=false
      drawAs=Image
      imageType=Vector
      mirroring=NoMirror
      resourceName=../../../Engine/Content/Slate/Starship/CoreWidgets/CheckBox/indeterminate.svg
      resourceObject=null
      tiling=NoTile
      imageSize=<16 16>
      margin=<ltrb 0 0 0 0>
      outlineSettings {
       bUseBrushTransparency=false
       roundingType=HalfHeightRadius
       width=0
       color {colorUseRule=UseColor_Specified, specifiedColor=<rgba 0 0 0 0>}
       cornerRadii=<0 0 0 0>
      }
      tintColor {colorUseRule=UseColor_Specified, specifiedColor=<rgba 0 0.162 0.7454 1>}
      uVRegion {bIsValid=false, max=<0 0>, min=<0 0>}
     }
     undeterminedPressedImage {
      bIsDynamicallyLoaded=false
      drawAs=Image
      imageType=Vector
      mirroring=NoMirror
      resourceName=../../../Engine/Content/Slate/Starship/CoreWidgets/CheckBox/indeterminate.svg
      resourceObject=null
      tiling=NoTile
      imageSize=<16 16>
      margin=<ltrb 0 0 0 0>
      outlineSettings {
       bUseBrushTransparency=false
       roundingType=HalfHeightRadius
       width=0
       color {colorUseRule=UseColor_Specified, specifiedColor=<rgba 0 0 0 0>}
       cornerRadii=<0 0 0 0>
      }
      tintColor {colorUseRule=UseColor_Specified, specifiedColor=<rgba 0 0.0802 0.3515 1>}
      uVRegion {bIsValid=false, max=<0 0>, min=<0 0>}
     }
    }
   }
   slot {class=HorizontalBoxSlot, properties {padding=<ltrb 2.5 2.5 2.5 2.5>}}
  }
  TextTag {
   widgetClassPath=&UEFN_TextBlock
   bIsVariable=true
   parent=HorizontalBox_54
   widget=TextTag
   properties {
    clipping=ClipToBounds
    text=Playback
    font {
     bForceMonospaced=false
     bMaterialIsStencil=false
     fontMaterial=null
     fontObject=&NotoSans
     letterSpacing=5
     monospacedWidth=1
     size=10
     skewAmount=0
     typefaceFontName=NotoSans-Regular
     outlineSettings {
      bApplyOutlineToDropShadows=false
      bMiteredCorners=false
      bSeparateFillAlpha=false
      outlineMaterial=null
      outlineSize=0
      outlineColor=<rgba 0 0 0 1>
     }
    }
   }
   slot {class=HorizontalBoxSlot, properties {horizontalAlignment=HAlign_Left, verticalAlignment=VAlign_Center, padding=<ltrb 2.5 0 0 0>}}
  }
  Border_Playback {
   widgetClassPath=&Border
   bIsVariable=true
   parent=Playback
   widget=Border_Playback
   properties {
    visibility=Collapsed
    background {
     bIsDynamicallyLoaded=false
     drawAs=Box
     imageType=NoImage
     mirroring=NoMirror
     resourceName=None
     resourceObject=null
     tiling=NoTile
     imageSize=<32 32>
     margin=<ltrb 0 0 0 0>
     outlineSettings {
      bUseBrushTransparency=false
      roundingType=HalfHeightRadius
      width=0
      color {colorUseRule=UseColor_Specified, specifiedColor=<rgba 0 0 0 0>}
      cornerRadii=<0 0 0 0>
     }
     tintColor {colorUseRule=UseColor_Specified, specifiedColor=<rgba 1 1 1 1>}
     uVRegion {bIsValid=false, max=<0 0>, min=<0 0>}
    }
    brushColor=<rgba 0 0 0 0.2>
    padding=<ltrb 20 0 0 0>
   }
   slot {class=VerticalBoxSlot, properties {padding=<ltrb 10 0 0 0>}}
  }
  PlaybackBox {
   widgetClassPath=&VerticalBox
   parent=Border_Playback
   widget=PlaybackBox
   slot {class=BorderSlot, properties {padding=<ltrb 20 0 0 0>}}
  }
  HorizontalBox_154 {widgetClassPath=&HorizontalBox, parent=PlaybackBox, widget=HorizontalBox_154, slot {class=VerticalBoxSlot}}
  PlaySelected {
   widgetClassPath=&Button
   parent=HorizontalBox_154
   widget=PlaySelected
   properties {widgetStyle {
     clickedSlateSound {resourceObject=null}
     disabled {
      bIsDynamicallyLoaded=false
      drawAs=Box
      imageType=NoImage
      mirroring=NoMirror
      resourceName=None
      resourceObject=null
      tiling=NoTile
      imageSize=<0 0>
      margin=<ltrb 0 0 0 0>
      outlineSettings {
       bUseBrushTransparency=false
       roundingType=FixedRadius
       width=1
       color {colorUseRule=UseColor_ColorTable, specifiedColor=<rgba 0 0 0 0>}
       cornerRadii=<4 4 4 4>
      }
      tintColor {colorUseRule=UseColor_Specified, specifiedColor=<rgba 0.0395 0.0395 0.0395 1>}
      uVRegion {bIsValid=false, max=<0 0>, min=<0 0>}
     }
     disabledForeground {colorUseRule=UseColor_ColorTable, specifiedColor=<rgba 0 0 0 0>}
     hovered {
      bIsDynamicallyLoaded=false
      drawAs=Box
      imageType=NoImage
      mirroring=NoMirror
      resourceName=None
      resourceObject=null
      tiling=NoTile
      imageSize=<32 32>
      margin=<ltrb 0 0 0 0>
      outlineSettings {
       bUseBrushTransparency=false
       roundingType=FixedRadius
       width=1
       color {colorUseRule=UseColor_ColorTable, specifiedColor=<rgba 0 0 0 0>}
       cornerRadii=<4 4 4 4>
      }
      tintColor {colorUseRule=UseColor_Specified, specifiedColor=<rgba 0.0953 0.0953 0.0953 1>}
      uVRegion {bIsValid=false, max=<0 0>, min=<0 0>}
     }
     hoveredForeground {colorUseRule=UseColor_ColorTable, specifiedColor=<rgba 0 0 0 0>}
     hoveredSlateSound {resourceObject=null}
     normal {
      bIsDynamicallyLoaded=false
      drawAs=Box
      imageType=NoImage
      mirroring=NoMirror
      resourceName=None
      resourceObject=null
      tiling=NoTile
      imageSize=<32 32>
      margin=<ltrb 0 0 0 0>
      outlineSettings {
       bUseBrushTransparency=false
       roundingType=FixedRadius
       width=1
       color {colorUseRule=UseColor_ColorTable, specifiedColor=<rgba 0 0 0 0>}
       cornerRadii=<4 4 4 4>
      }
      tintColor {colorUseRule=UseColor_Specified, specifiedColor=<rgba 0.0423 0.0349 0.1719 0.801>}
      uVRegion {bIsValid=false, max=<0 0>, min=<0 0>}
     }
     normalForeground {colorUseRule=UseColor_ColorTable, specifiedColor=<rgba 0 0 0 0>}
     normalPadding=<ltrb 12 1.5 12 1.5>
     pressed {
      bIsDynamicallyLoaded=false
      drawAs=Box
      imageType=NoImage
      mirroring=NoMirror
      resourceName=None
      resourceObject=null
      tiling=NoTile
      imageSize=<32 32>
      margin=<ltrb 0 0 0 0>
      outlineSettings {
       bUseBrushTransparency=false
       roundingType=FixedRadius
       width=1
       color {colorUseRule=UseColor_ColorTable, specifiedColor=<rgba 0 0 0 0>}
       cornerRadii=<4 4 4 4>
      }
      tintColor {colorUseRule=UseColor_Specified, specifiedColor=<rgba 0.0284 0.0284 0.0284 1>}
      uVRegion {bIsValid=false, max=<0 0>, min=<0 0>}
     }
     pressedForeground {colorUseRule=UseColor_ColorTable, specifiedColor=<rgba 0 0 0 0>}
     pressedPadding=<ltrb 12 2.5 12 0.5>
     pressedSlateSound {resourceObject=null}
    }}
   slot {class=HorizontalBoxSlot, properties {padding=<ltrb 0.5 0 0.5 0>}}
  }
  Text_5 {
   widgetClassPath=&UEFN_TextBlock
   bIsVariable=true
   parent=PlaySelected
   widget=Text_5
   properties {
    clipping=ClipToBounds
    text=Play Selected Sounds
    font {
     bForceMonospaced=false
     bMaterialIsStencil=false
     fontMaterial=null
     fontObject=&NotoSans
     letterSpacing=5
     monospacedWidth=1
     size=10
     skewAmount=0
     typefaceFontName=NotoSans-Regular
     outlineSettings {
      bApplyOutlineToDropShadows=false
      bMiteredCorners=false
      bSeparateFillAlpha=false
      outlineMaterial=null
      outlineSize=0
      outlineColor=<rgba 0 0 0 1>
     }
    }
   }
   slot {class=ButtonSlot, properties {padding=<ltrb 0 0 0 0>}}
  }
  StopSelected {
   widgetClassPath=&Button
   parent=HorizontalBox_154
   widget=StopSelected
   properties {widgetStyle {
     clickedSlateSound {resourceObject=null}
     disabled {
      bIsDynamicallyLoaded=false
      drawAs=Box
      imageType=NoImage
      mirroring=NoMirror
      resourceName=None
      resourceObject=null
      tiling=NoTile
      imageSize=<0 0>
      margin=<ltrb 0 0 0 0>
      outlineSettings {
       bUseBrushTransparency=false
       roundingType=FixedRadius
       width=1
       color {colorUseRule=UseColor_ColorTable, specifiedColor=<rgba 0 0 0 0>}
       cornerRadii=<4 4 4 4>
      }
      tintColor {colorUseRule=UseColor_Specified, specifiedColor=<rgba 0.0395 0.0395 0.0395 1>}
      uVRegion {bIsValid=false, max=<0 0>, min=<0 0>}
     }
     disabledForeground {colorUseRule=UseColor_ColorTable, specifiedColor=<rgba 0 0 0 0>}
     hovered {
      bIsDynamicallyLoaded=false
      drawAs=Box
      imageType=NoImage
      mirroring=NoMirror
      resourceName=None
      resourceObject=null
      tiling=NoTile
      imageSize=<32 32>
      margin=<ltrb 0 0 0 0>
      outlineSettings {
       bUseBrushTransparency=false
       roundingType=FixedRadius
       width=1
       color {colorUseRule=UseColor_ColorTable, specifiedColor=<rgba 0 0 0 0>}
       cornerRadii=<4 4 4 4>
      }
      tintColor {colorUseRule=UseColor_Specified, specifiedColor=<rgba 0.0953 0.003 0.003 1>}
      uVRegion {bIsValid=false, max=<0 0>, min=<0 0>}
     }
     hoveredForeground {colorUseRule=UseColor_ColorTable, specifiedColor=<rgba 0 0 0 0>}
     hoveredSlateSound {resourceObject=null}
     normal {
      bIsDynamicallyLoaded=false
      drawAs=Box
      imageType=NoImage
      mirroring=NoMirror
      resourceName=None
      resourceObject=null
      tiling=NoTile
      imageSize=<32 32>
      margin=<ltrb 0 0 0 0>
      outlineSettings {
       bUseBrushTransparency=false
       roundingType=FixedRadius
       width=1
       color {colorUseRule=UseColor_ColorTable, specifiedColor=<rgba 0 0 0 0>}
       cornerRadii=<4 4 4 4>
      }
      tintColor {colorUseRule=UseColor_Specified, specifiedColor=<rgba 0.1719 0.0151 0.0322 0.801>}
      uVRegion {bIsValid=false, max=<0 0>, min=<0 0>}
     }
     normalForeground {colorUseRule=UseColor_ColorTable, specifiedColor=<rgba 0 0 0 0>}
     normalPadding=<ltrb 12 1.5 12 1.5>
     pressed {
      bIsDynamicallyLoaded=false
      drawAs=Box
      imageType=NoImage
      mirroring=NoMirror
      resourceName=None
      resourceObject=null
      tiling=NoTile
      imageSize=<32 32>
      margin=<ltrb 0 0 0 0>
      outlineSettings {
       bUseBrushTransparency=false
       roundingType=FixedRadius
       width=1
       color {colorUseRule=UseColor_ColorTable, specifiedColor=<rgba 0 0 0 0>}
       cornerRadii=<4 4 4 4>
      }
      tintColor {colorUseRule=UseColor_Specified, specifiedColor=<rgba 0.0284 0 0 1>}
      uVRegion {bIsValid=false, max=<0 0>, min=<0 0>}
     }
     pressedForeground {colorUseRule=UseColor_ColorTable, specifiedColor=<rgba 0 0 0 0>}
     pressedPadding=<ltrb 12 2.5 12 0.5>
     pressedSlateSound {resourceObject=null}
    }}
   slot {class=HorizontalBoxSlot, properties {padding=<ltrb 0.5 0 0.5 0>}}
  }
  Text_6 {
   widgetClassPath=&UEFN_TextBlock
   bIsVariable=true
   parent=StopSelected
   widget=Text_6
   properties {
    clipping=ClipToBounds
    text=Stop Sounds
    font {
     bForceMonospaced=false
     bMaterialIsStencil=false
     fontMaterial=null
     fontObject=&NotoSans
     letterSpacing=5
     monospacedWidth=1
     size=10
     skewAmount=0
     typefaceFontName=NotoSans-Regular
     outlineSettings {
      bApplyOutlineToDropShadows=false
      bMiteredCorners=false
      bSeparateFillAlpha=false
      outlineMaterial=null
      outlineSize=0
      outlineColor=<rgba 0 0 0 1>
     }
    }
   }
   slot {class=ButtonSlot, properties {padding=<ltrb 0 0 0 0>}}
  }
  DrawDebugBox {widgetClassPath=&HorizontalBox, parent=PlaybackBox, widget=DrawDebugBox, slot {class=VerticalBoxSlot}}
  DrawDebugLabel {
   widgetClassPath=&TextBlock
   parent=DrawDebugBox
   widget=DrawDebugLabel
   properties {
    text=Draw Debug Sphere
    font {
     bForceMonospaced=false
     bMaterialIsStencil=false
     fontMaterial=null
     fontObject=&NotoSans
     letterSpacing=5
     monospacedWidth=1
     size=7.5
     skewAmount=0
     typefaceFontName=NotoSans-Regular
     outlineSettings {
      bApplyOutlineToDropShadows=false
      bMiteredCorners=false
      bSeparateFillAlpha=false
      outlineMaterial=null
      outlineSize=0
      outlineColor=<rgba 0 0 0 1>
     }
    }
   }
   slot {class=HorizontalBoxSlot, properties {verticalAlignment=VAlign_Center}}
  }
  DrawDebug {
   widgetClassPath=&EditorUtilityCheckBox
   bIsVariable=true
   parent=DrawDebugBox
   widget=DrawDebug
   slot {class=HorizontalBoxSlot, properties {padding=<ltrb 5 5 5 5>}}
  }
  Distance {widgetClassPath=&HorizontalBox, parent=PlaybackBox, widget=Distance, slot {class=VerticalBoxSlot}}
  DistanceLabel {
   widgetClassPath=&TextBlock
   parent=Distance
   widget=DistanceLabel
   properties {
    text=Distance
    font {
     bForceMonospaced=false
     bMaterialIsStencil=false
     fontMaterial=null
     fontObject=&NotoSans
     letterSpacing=5
     monospacedWidth=1
     size=7.5
     skewAmount=0
     typefaceFontName=NotoSans-Regular
     outlineSettings {
      bApplyOutlineToDropShadows=false
      bMiteredCorners=false
      bSeparateFillAlpha=false
      outlineMaterial=null
      outlineSize=0
      outlineColor=<rgba 0 0 0 1>
     }
    }
   }
   slot {class=HorizontalBoxSlot, properties {verticalAlignment=VAlign_Center}}
  }
  Distance_ComboBox {
   widgetClassPath=/Script/Blutility.EditorUtilityComboBoxString
   bIsVariable=true
   parent=Distance
   widget=Distance_ComboBox
   properties {
    selectedOption=0
    defaultOptions [0,500,1000,1500,2000,4000,6000,8000]
    font {
     bForceMonospaced=false
     bMaterialIsStencil=false
     fontMaterial=null
     fontObject=&NotoSans
     letterSpacing=5
     monospacedWidth=1
     size=7.5
     skewAmount=0
     typefaceFontName=NotoSans-Regular
     outlineSettings {
      bApplyOutlineToDropShadows=false
      bMiteredCorners=false
      bSeparateFillAlpha=false
      outlineMaterial=null
      outlineSize=0
      outlineColor=<rgba 0 0 0 1>
     }
    }
    foregroundColor {colorUseRule=UseColor_Specified, specifiedColor=<rgba 0.5271 0.5271 0.5271 1>}
   }
   slot {class=HorizontalBoxSlot, properties {padding=<ltrb 5 5 5 5>}}
  }
  Affiliation {widgetClassPath=&HorizontalBox, parent=PlaybackBox, widget=Affiliation, slot {class=VerticalBoxSlot}}
  AffiliationLabel {
   widgetClassPath=&TextBlock
   parent=Affiliation
   widget=AffiliationLabel
   properties {
    text=Affiliation
    font {
     bForceMonospaced=false
     bMaterialIsStencil=false
     fontMaterial=null
     fontObject=&NotoSans
     letterSpacing=5
     monospacedWidth=1
     size=7.5
     skewAmount=0
     typefaceFontName=NotoSans-Regular
     outlineSettings {
      bApplyOutlineToDropShadows=false
      bMiteredCorners=false
      bSeparateFillAlpha=false
      outlineMaterial=null
      outlineSize=0
      outlineColor=<rgba 0 0 0 1>
     }
    }
   }
   slot {class=HorizontalBoxSlot, properties {verticalAlignment=VAlign_Center}}
  }
  SoundAffiliation {
   widgetClassPath=&EditorUtilityComboBoxString
   bIsVariable=true
   parent=Affiliation
   widget=SoundAffiliation
   properties {
    selectedOption=Default
    defaultOptions [Default,Local,Enemy,Friendly]
    font {
     bForceMonospaced=false
     bMaterialIsStencil=false
     fontMaterial=null
     fontObject=&NotoSans
     letterSpacing=5
     monospacedWidth=1
     size=7.5
     skewAmount=0
     typefaceFontName=NotoSans-Regular
     outlineSettings {
      bApplyOutlineToDropShadows=false
      bMiteredCorners=false
      bSeparateFillAlpha=false
      outlineMaterial=null
      outlineSize=0
      outlineColor=<rgba 0 0 0 1>
     }
    }
    foregroundColor {colorUseRule=UseColor_Specified, specifiedColor=<rgba 0.5271 0.5271 0.5271 1>}
   }
   slot {class=HorizontalBoxSlot, properties {padding=<ltrb 5 5 5 5>}}
  }
  ScrollBox {
   widgetClassPath=&ScrollBox
   parent=VerticalBox_2
   widget=ScrollBox
   properties {
    widgetBarStyle {
     thickness=8
     draggedThumbImage {
      bIsDynamicallyLoaded=false
      drawAs=RoundedBox
      imageType=NoImage
      mirroring=NoMirror
      resourceName=None
      resourceObject=null
      tiling=NoTile
      imageSize=<0 0>
      margin=<ltrb 0 0 0 0>
      outlineSettings {
       bUseBrushTransparency=false
       roundingType=FixedRadius
       width=0
       color {colorUseRule=UseColor_Specified, specifiedColor=<rgba 0 0 0 0>}
       cornerRadii=<4 4 4 4>
      }
      tintColor {colorUseRule=UseColor_Specified, specifiedColor=<rgba 0.2159 0.2159 0.2159 1>}
      uVRegion {bIsValid=false, max=<0 0>, min=<0 0>}
     }
     horizontalBackgroundImage {
      bIsDynamicallyLoaded=false
      drawAs=NoDrawType
      imageType=NoImage
      mirroring=NoMirror
      resourceName=None
      resourceObject=null
      tiling=NoTile
      imageSize=<0 0>
      margin=<ltrb 0 0 0 0>
      outlineSettings {
       bUseBrushTransparency=false
       roundingType=HalfHeightRadius
       width=0
       color {colorUseRule=UseColor_Specified, specifiedColor=<rgba 0 0 0 0>}
       cornerRadii=<0 0 0 0>
      }
      tintColor {colorUseRule=UseColor_Specified, specifiedColor=<rgba 1 1 1 1>}
      uVRegion {bIsValid=false, max=<0 0>, min=<0 0>}
     }
     horizontalBottomSlotImage {
      bIsDynamicallyLoaded=false
      drawAs=NoDrawType
      imageType=NoImage
      mirroring=NoMirror
      resourceName=None
      resourceObject=null
      tiling=NoTile
      imageSize=<0 0>
      margin=<ltrb 0 0 0 0>
      outlineSettings {
       bUseBrushTransparency=false
       roundingType=HalfHeightRadius
       width=0
       color {colorUseRule=UseColor_Specified, specifiedColor=<rgba 0 0 0 0>}
       cornerRadii=<0 0 0 0>
      }
      tintColor {colorUseRule=UseColor_Specified, specifiedColor=<rgba 1 1 1 1>}
      uVRegion {bIsValid=false, max=<0 0>, min=<0 0>}
     }
     horizontalTopSlotImage {
      bIsDynamicallyLoaded=false
      drawAs=NoDrawType
      imageType=NoImage
      mirroring=NoMirror
      resourceName=None
      resourceObject=null
      tiling=NoTile
      imageSize=<0 0>
      margin=<ltrb 0 0 0 0>
      outlineSettings {
       bUseBrushTransparency=false
       roundingType=HalfHeightRadius
       width=0
       color {colorUseRule=UseColor_Specified, specifiedColor=<rgba 0 0 0 0>}
       cornerRadii=<0 0 0 0>
      }
      tintColor {colorUseRule=UseColor_Specified, specifiedColor=<rgba 1 1 1 1>}
      uVRegion {bIsValid=false, max=<0 0>, min=<0 0>}
     }
     hoveredThumbImage {
      bIsDynamicallyLoaded=false
      drawAs=RoundedBox
      imageType=NoImage
      mirroring=NoMirror
      resourceName=None
      resourceObject=null
      tiling=NoTile
      imageSize=<0 0>
      margin=<ltrb 0 0 0 0>
      outlineSettings {
       bUseBrushTransparency=false
       roundingType=FixedRadius
       width=0
       color {colorUseRule=UseColor_Specified, specifiedColor=<rgba 0 0 0 0>}
       cornerRadii=<4 4 4 4>
      }
      tintColor {colorUseRule=UseColor_Specified, specifiedColor=<rgba 0.2159 0.2159 0.2159 1>}
      uVRegion {bIsValid=false, max=<0 0>, min=<0 0>}
     }
     normalThumbImage {
      bIsDynamicallyLoaded=false
      drawAs=RoundedBox
      imageType=NoImage
      mirroring=NoMirror
      resourceName=None
      resourceObject=null
      tiling=NoTile
      imageSize=<0 0>
      margin=<ltrb 0 0 0 0>
      outlineSettings {
       bUseBrushTransparency=false
       roundingType=FixedRadius
       width=0
       color {colorUseRule=UseColor_Specified, specifiedColor=<rgba 0 0 0 0>}
       cornerRadii=<4 4 4 4>
      }
      tintColor {colorUseRule=UseColor_Specified, specifiedColor=<rgba 0.0953 0.0953 0.0953 1>}
      uVRegion {bIsValid=false, max=<0 0>, min=<0 0>}
     }
     verticalBackgroundImage {
      bIsDynamicallyLoaded=false
      drawAs=NoDrawType
      imageType=NoImage
      mirroring=NoMirror
      resourceName=None
      resourceObject=null
      tiling=NoTile
      imageSize=<0 0>
      margin=<ltrb 0 0 0 0>
      outlineSettings {
       bUseBrushTransparency=false
       roundingType=HalfHeightRadius
       width=0
       color {colorUseRule=UseColor_Specified, specifiedColor=<rgba 0 0 0 0>}
       cornerRadii=<0 0 0 0>
      }
      tintColor {colorUseRule=UseColor_Specified, specifiedColor=<rgba 1 1 1 1>}
      uVRegion {bIsValid=false, max=<0 0>, min=<0 0>}
     }
     verticalBottomSlotImage {
      bIsDynamicallyLoaded=false
      drawAs=NoDrawType
      imageType=NoImage
      mirroring=NoMirror
      resourceName=None
      resourceObject=null
      tiling=NoTile
      imageSize=<0 0>
      margin=<ltrb 0 0 0 0>
      outlineSettings {
       bUseBrushTransparency=false
       roundingType=HalfHeightRadius
       width=0
       color {colorUseRule=UseColor_Specified, specifiedColor=<rgba 0 0 0 0>}
       cornerRadii=<0 0 0 0>
      }
      tintColor {colorUseRule=UseColor_Specified, specifiedColor=<rgba 1 1 1 1>}
      uVRegion {bIsValid=false, max=<0 0>, min=<0 0>}
     }
     verticalTopSlotImage {
      bIsDynamicallyLoaded=false
      drawAs=NoDrawType
      imageType=NoImage
      mirroring=NoMirror
      resourceName=None
      resourceObject=null
      tiling=NoTile
      imageSize=<0 0>
      margin=<ltrb 0 0 0 0>
      outlineSettings {
       bUseBrushTransparency=false
       roundingType=HalfHeightRadius
       width=0
       color {colorUseRule=UseColor_Specified, specifiedColor=<rgba 0 0 0 0>}
       cornerRadii=<0 0 0 0>
      }
      tintColor {colorUseRule=UseColor_Specified, specifiedColor=<rgba 1 1 1 1>}
      uVRegion {bIsValid=false, max=<0 0>, min=<0 0>}
     }
    }
    widgetStyle {
     barThickness=8
     bottomShadowBrush {
      bIsDynamicallyLoaded=false
      drawAs=Box
      imageType=FullColor
      mirroring=NoMirror
      resourceName=../../../Engine/Content/Slate/Common/ScrollBoxShadowBottom.png
      resourceObject=null
      tiling=NoTile
      imageSize=<16 8>
      margin=<ltrb 0.5 0 0.5 1>
      outlineSettings {
       bUseBrushTransparency=false
       roundingType=HalfHeightRadius
       width=0
       color {colorUseRule=UseColor_Specified, specifiedColor=<rgba 0 0 0 0>}
       cornerRadii=<0 0 0 0>
      }
      tintColor {colorUseRule=UseColor_Specified, specifiedColor=<rgba 1 1 1 1>}
      uVRegion {bIsValid=false, max=<0 0>, min=<0 0>}
     }
     horizontalScrolledContentPadding=<ltrb 0 0 1 0>
     leftShadowBrush {
      bIsDynamicallyLoaded=false
      drawAs=Box
      imageType=FullColor
      mirroring=NoMirror
      resourceName=../../../Engine/Content/Slate/Common/ScrollBoxShadowLeft.png
      resourceObject=null
      tiling=NoTile
      imageSize=<8 16>
      margin=<ltrb 1 0.5 0 0.5>
      outlineSettings {
       bUseBrushTransparency=false
       roundingType=HalfHeightRadius
       width=0
       color {colorUseRule=UseColor_Specified, specifiedColor=<rgba 0 0 0 0>}
       cornerRadii=<0 0 0 0>
      }
      tintColor {colorUseRule=UseColor_Specified, specifiedColor=<rgba 1 1 1 1>}
      uVRegion {bIsValid=false, max=<0 0>, min=<0 0>}
     }
     rightShadowBrush {
      bIsDynamicallyLoaded=false
      drawAs=Box
      imageType=FullColor
      mirroring=NoMirror
      resourceName=../../../Engine/Content/Slate/Common/ScrollBoxShadowRight.png
      resourceObject=null
      tiling=NoTile
      imageSize=<8 16>
      margin=<ltrb 0 0.5 1 0.5>
      outlineSettings {
       bUseBrushTransparency=false
       roundingType=HalfHeightRadius
       width=0
       color {colorUseRule=UseColor_Specified, specifiedColor=<rgba 0 0 0 0>}
       cornerRadii=<0 0 0 0>
      }
      tintColor {colorUseRule=UseColor_Specified, specifiedColor=<rgba 1 1 1 1>}
      uVRegion {bIsValid=false, max=<0 0>, min=<0 0>}
     }
     topShadowBrush {
      bIsDynamicallyLoaded=false
      drawAs=Box
      imageType=FullColor
      mirroring=NoMirror
      resourceName=../../../Engine/Content/Slate/Common/ScrollBoxShadowTop.png
      resourceObject=null
      tiling=NoTile
      imageSize=<16 8>
      margin=<ltrb 0.5 1 0.5 0>
      outlineSettings {
       bUseBrushTransparency=false
       roundingType=HalfHeightRadius
       width=0
       color {colorUseRule=UseColor_Specified, specifiedColor=<rgba 0 0 0 0>}
       cornerRadii=<0 0 0 0>
      }
      tintColor {colorUseRule=UseColor_Specified, specifiedColor=<rgba 1 1 1 1>}
      uVRegion {bIsValid=false, max=<0 0>, min=<0 0>}
     }
     verticalScrolledContentPadding=<ltrb 0 0 0 1>
    }
   }
   slot {class=VerticalBoxSlot, properties {size {sizeRule=Fill, value=1}}}
  }
  SoundPropertyListView {
   widgetClassPath=&CommonListView
   bIsVariable=true
   parent=ScrollBox
   widget=SoundPropertyListView
   properties {
    entryWidgetClass=&/EUW_SoundLibraryEditor_EditEntry.EUW_SoundLibraryEditor_EditEntry_C
    selectionMode=None
    scrollBarStyle {
     thickness=0
     draggedThumbImage {
      bIsDynamicallyLoaded=false
      drawAs=RoundedBox
      imageType=NoImage
      mirroring=NoMirror
      resourceName=None
      resourceObject=null
      tiling=NoTile
      imageSize=<0 0>
      margin=<ltrb 0 0 0 0>
      outlineSettings {
       bUseBrushTransparency=false
       roundingType=FixedRadius
       width=0
       color {colorUseRule=UseColor_Specified, specifiedColor=<rgba 0 0 0 0>}
       cornerRadii=<4 4 4 4>
      }
      tintColor {colorUseRule=UseColor_Specified, specifiedColor=<rgba 0.2159 0.2159 0.2159 1>}
      uVRegion {bIsValid=false, max=<0 0>, min=<0 0>}
     }
     horizontalBackgroundImage {
      bIsDynamicallyLoaded=false
      drawAs=NoDrawType
      imageType=NoImage
      mirroring=NoMirror
      resourceName=None
      resourceObject=null
      tiling=NoTile
      imageSize=<0 0>
      margin=<ltrb 0 0 0 0>
      outlineSettings {
       bUseBrushTransparency=false
       roundingType=HalfHeightRadius
       width=0
       color {colorUseRule=UseColor_Specified, specifiedColor=<rgba 0 0 0 0>}
       cornerRadii=<0 0 0 0>
      }
      tintColor {colorUseRule=UseColor_Specified, specifiedColor=<rgba 1 1 1 1>}
      uVRegion {bIsValid=false, max=<0 0>, min=<0 0>}
     }
     horizontalBottomSlotImage {
      bIsDynamicallyLoaded=false
      drawAs=NoDrawType
      imageType=NoImage
      mirroring=NoMirror
      resourceName=None
      resourceObject=null
      tiling=NoTile
      imageSize=<0 0>
      margin=<ltrb 0 0 0 0>
      outlineSettings {
       bUseBrushTransparency=false
       roundingType=HalfHeightRadius
       width=0
       color {colorUseRule=UseColor_Specified, specifiedColor=<rgba 0 0 0 0>}
       cornerRadii=<0 0 0 0>
      }
      tintColor {colorUseRule=UseColor_Specified, specifiedColor=<rgba 1 1 1 1>}
      uVRegion {bIsValid=false, max=<0 0>, min=<0 0>}
     }
     horizontalTopSlotImage {
      bIsDynamicallyLoaded=false
      drawAs=NoDrawType
      imageType=NoImage
      mirroring=NoMirror
      resourceName=None
      resourceObject=null
      tiling=NoTile
      imageSize=<0 0>
      margin=<ltrb 0 0 0 0>
      outlineSettings {
       bUseBrushTransparency=false
       roundingType=HalfHeightRadius
       width=0
       color {colorUseRule=UseColor_Specified, specifiedColor=<rgba 0 0 0 0>}
       cornerRadii=<0 0 0 0>
      }
      tintColor {colorUseRule=UseColor_Specified, specifiedColor=<rgba 1 1 1 1>}
      uVRegion {bIsValid=false, max=<0 0>, min=<0 0>}
     }
     hoveredThumbImage {
      bIsDynamicallyLoaded=false
      drawAs=RoundedBox
      imageType=NoImage
      mirroring=NoMirror
      resourceName=None
      resourceObject=null
      tiling=NoTile
      imageSize=<0 0>
      margin=<ltrb 0 0 0 0>
      outlineSettings {
       bUseBrushTransparency=false
       roundingType=FixedRadius
       width=0
       color {colorUseRule=UseColor_Specified, specifiedColor=<rgba 0 0 0 0>}
       cornerRadii=<4 4 4 4>
      }
      tintColor {colorUseRule=UseColor_Specified, specifiedColor=<rgba 0.2159 0.2159 0.2159 1>}
      uVRegion {bIsValid=false, max=<0 0>, min=<0 0>}
     }
     normalThumbImage {
      bIsDynamicallyLoaded=false
      drawAs=RoundedBox
      imageType=NoImage
      mirroring=NoMirror
      resourceName=None
      resourceObject=null
      tiling=NoTile
      imageSize=<0 0>
      margin=<ltrb 0 0 0 0>
      outlineSettings {
       bUseBrushTransparency=false
       roundingType=FixedRadius
       width=0
       color {colorUseRule=UseColor_Specified, specifiedColor=<rgba 0 0 0 0>}
       cornerRadii=<4 4 4 4>
      }
      tintColor {colorUseRule=UseColor_Specified, specifiedColor=<rgba 0.0953 0.0953 0.0953 1>}
      uVRegion {bIsValid=false, max=<0 0>, min=<0 0>}
     }
     verticalBackgroundImage {
      bIsDynamicallyLoaded=false
      drawAs=NoDrawType
      imageType=NoImage
      mirroring=NoMirror
      resourceName=None
      resourceObject=null
      tiling=NoTile
      imageSize=<0 0>
      margin=<ltrb 0 0 0 0>
      outlineSettings {
       bUseBrushTransparency=false
       roundingType=HalfHeightRadius
       width=0
       color {colorUseRule=UseColor_Specified, specifiedColor=<rgba 0 0 0 0>}
       cornerRadii=<0 0 0 0>
      }
      tintColor {colorUseRule=UseColor_Specified, specifiedColor=<rgba 1 1 1 1>}
      uVRegion {bIsValid=false, max=<0 0>, min=<0 0>}
     }
     verticalBottomSlotImage {
      bIsDynamicallyLoaded=false
      drawAs=NoDrawType
      imageType=NoImage
      mirroring=NoMirror
      resourceName=None
      resourceObject=null
      tiling=NoTile
      imageSize=<0 0>
      margin=<ltrb 0 0 0 0>
      outlineSettings {
       bUseBrushTransparency=false
       roundingType=HalfHeightRadius
       width=0
       color {colorUseRule=UseColor_Specified, specifiedColor=<rgba 0 0 0 0>}
       cornerRadii=<0 0 0 0>
      }
      tintColor {colorUseRule=UseColor_Specified, specifiedColor=<rgba 1 1 1 1>}
      uVRegion {bIsValid=false, max=<0 0>, min=<0 0>}
     }
     verticalTopSlotImage {
      bIsDynamicallyLoaded=false
      drawAs=NoDrawType
      imageType=NoImage
      mirroring=NoMirror
      resourceName=None
      resourceObject=null
      tiling=NoTile
      imageSize=<0 0>
      margin=<ltrb 0 0 0 0>
      outlineSettings {
       bUseBrushTransparency=false
       roundingType=HalfHeightRadius
       width=0
       color {colorUseRule=UseColor_Specified, specifiedColor=<rgba 0 0 0 0>}
       cornerRadii=<0 0 0 0>
      }
      tintColor {colorUseRule=UseColor_Specified, specifiedColor=<rgba 1 1 1 1>}
      uVRegion {bIsValid=false, max=<0 0>, min=<0 0>}
     }
    }
   }
   slot {class=ScrollBoxSlot, properties {padding=<ltrb 10 10 10 10>}}
  }
  VerticalBox {widgetClassPath=&VerticalBox, parent=WidgetSwitcher, widget=VerticalBox, slot {class=WidgetSwitcherSlot}}
  ScrollBox_0 {
   widgetClassPath=&ScrollBox
   parent=VerticalBox
   widget=ScrollBox_0
   properties {
    widgetBarStyle {
     thickness=8
     draggedThumbImage {
      bIsDynamicallyLoaded=false
      drawAs=RoundedBox
      imageType=NoImage
      mirroring=NoMirror
      resourceName=None
      resourceObject=null
      tiling=NoTile
      imageSize=<0 0>
      margin=<ltrb 0 0 0 0>
      outlineSettings {
       bUseBrushTransparency=false
       roundingType=FixedRadius
       width=0
       color {colorUseRule=UseColor_Specified, specifiedColor=<rgba 0 0 0 0>}
       cornerRadii=<4 4 4 4>
      }
      tintColor {colorUseRule=UseColor_Specified, specifiedColor=<rgba 0.2159 0.2159 0.2159 1>}
      uVRegion {bIsValid=false, max=<0 0>, min=<0 0>}
     }
     horizontalBackgroundImage {
      bIsDynamicallyLoaded=false
      drawAs=NoDrawType
      imageType=NoImage
      mirroring=NoMirror
      resourceName=None
      resourceObject=null
      tiling=NoTile
      imageSize=<0 0>
      margin=<ltrb 0 0 0 0>
      outlineSettings {
       bUseBrushTransparency=false
       roundingType=HalfHeightRadius
       width=0
       color {colorUseRule=UseColor_Specified, specifiedColor=<rgba 0 0 0 0>}
       cornerRadii=<0 0 0 0>
      }
      tintColor {colorUseRule=UseColor_Specified, specifiedColor=<rgba 1 1 1 1>}
      uVRegion {bIsValid=false, max=<0 0>, min=<0 0>}
     }
     horizontalBottomSlotImage {
      bIsDynamicallyLoaded=false
      drawAs=NoDrawType
      imageType=NoImage
      mirroring=NoMirror
      resourceName=None
      resourceObject=null
      tiling=NoTile
      imageSize=<0 0>
      margin=<ltrb 0 0 0 0>
      outlineSettings {
       bUseBrushTransparency=false
       roundingType=HalfHeightRadius
       width=0
       color {colorUseRule=UseColor_Specified, specifiedColor=<rgba 0 0 0 0>}
       cornerRadii=<0 0 0 0>
      }
      tintColor {colorUseRule=UseColor_Specified, specifiedColor=<rgba 1 1 1 1>}
      uVRegion {bIsValid=false, max=<0 0>, min=<0 0>}
     }
     horizontalTopSlotImage {
      bIsDynamicallyLoaded=false
      drawAs=NoDrawType
      imageType=NoImage
      mirroring=NoMirror
      resourceName=None
      resourceObject=null
      tiling=NoTile
      imageSize=<0 0>
      margin=<ltrb 0 0 0 0>
      outlineSettings {
       bUseBrushTransparency=false
       roundingType=HalfHeightRadius
       width=0
       color {colorUseRule=UseColor_Specified, specifiedColor=<rgba 0 0 0 0>}
       cornerRadii=<0 0 0 0>
      }
      tintColor {colorUseRule=UseColor_Specified, specifiedColor=<rgba 1 1 1 1>}
      uVRegion {bIsValid=false, max=<0 0>, min=<0 0>}
     }
     hoveredThumbImage {
      bIsDynamicallyLoaded=false
      drawAs=RoundedBox
      imageType=NoImage
      mirroring=NoMirror
      resourceName=None
      resourceObject=null
      tiling=NoTile
      imageSize=<0 0>
      margin=<ltrb 0 0 0 0>
      outlineSettings {
       bUseBrushTransparency=false
       roundingType=FixedRadius
       width=0
       color {colorUseRule=UseColor_Specified, specifiedColor=<rgba 0 0 0 0>}
       cornerRadii=<4 4 4 4>
      }
      tintColor {colorUseRule=UseColor_Specified, specifiedColor=<rgba 0.2159 0.2159 0.2159 1>}
      uVRegion {bIsValid=false, max=<0 0>, min=<0 0>}
     }
     normalThumbImage {
      bIsDynamicallyLoaded=false
      drawAs=RoundedBox
      imageType=NoImage
      mirroring=NoMirror
      resourceName=None
      resourceObject=null
      tiling=NoTile
      imageSize=<0 0>
      margin=<ltrb 0 0 0 0>
      outlineSettings {
       bUseBrushTransparency=false
       roundingType=FixedRadius
       width=0
       color {colorUseRule=UseColor_Specified, specifiedColor=<rgba 0 0 0 0>}
       cornerRadii=<4 4 4 4>
      }
      tintColor {colorUseRule=UseColor_Specified, specifiedColor=<rgba 0.0953 0.0953 0.0953 1>}
      uVRegion {bIsValid=false, max=<0 0>, min=<0 0>}
     }
     verticalBackgroundImage {
      bIsDynamicallyLoaded=false
      drawAs=NoDrawType
      imageType=NoImage
      mirroring=NoMirror
      resourceName=None
      resourceObject=null
      tiling=NoTile
      imageSize=<0 0>
      margin=<ltrb 0 0 0 0>
      outlineSettings {
       bUseBrushTransparency=false
       roundingType=HalfHeightRadius
       width=0
       color {colorUseRule=UseColor_Specified, specifiedColor=<rgba 0 0 0 0>}
       cornerRadii=<0 0 0 0>
      }
      tintColor {colorUseRule=UseColor_Specified, specifiedColor=<rgba 1 1 1 1>}
      uVRegion {bIsValid=false, max=<0 0>, min=<0 0>}
     }
     verticalBottomSlotImage {
      bIsDynamicallyLoaded=false
      drawAs=NoDrawType
      imageType=NoImage
      mirroring=NoMirror
      resourceName=None
      resourceObject=null
      tiling=NoTile
      imageSize=<0 0>
      margin=<ltrb 0 0 0 0>
      outlineSettings {
       bUseBrushTransparency=false
       roundingType=HalfHeightRadius
       width=0
       color {colorUseRule=UseColor_Specified, specifiedColor=<rgba 0 0 0 0>}
       cornerRadii=<0 0 0 0>
      }
      tintColor {colorUseRule=UseColor_Specified, specifiedColor=<rgba 1 1 1 1>}
      uVRegion {bIsValid=false, max=<0 0>, min=<0 0>}
     }
     verticalTopSlotImage {
      bIsDynamicallyLoaded=false
      drawAs=NoDrawType
      imageType=NoImage
      mirroring=NoMirror
      resourceName=None
      resourceObject=null
      tiling=NoTile
      imageSize=<0 0>
      margin=<ltrb 0 0 0 0>
      outlineSettings {
       bUseBrushTransparency=false
       roundingType=HalfHeightRadius
       width=0
       color {colorUseRule=UseColor_Specified, specifiedColor=<rgba 0 0 0 0>}
       cornerRadii=<0 0 0 0>
      }
      tintColor {colorUseRule=UseColor_Specified, specifiedColor=<rgba 1 1 1 1>}
      uVRegion {bIsValid=false, max=<0 0>, min=<0 0>}
     }
    }
    widgetStyle {
     barThickness=8
     bottomShadowBrush {
      bIsDynamicallyLoaded=false
      drawAs=Box
      imageType=FullColor
      mirroring=NoMirror
      resourceName=../../../Engine/Content/Slate/Common/ScrollBoxShadowBottom.png
      resourceObject=null
      tiling=NoTile
      imageSize=<16 8>
      margin=<ltrb 0.5 0 0.5 1>
      outlineSettings {
       bUseBrushTransparency=false
       roundingType=HalfHeightRadius
       width=0
       color {colorUseRule=UseColor_Specified, specifiedColor=<rgba 0 0 0 0>}
       cornerRadii=<0 0 0 0>
      }
      tintColor {colorUseRule=UseColor_Specified, specifiedColor=<rgba 1 1 1 1>}
      uVRegion {bIsValid=false, max=<0 0>, min=<0 0>}
     }
     horizontalScrolledContentPadding=<ltrb 0 0 1 0>
     leftShadowBrush {
      bIsDynamicallyLoaded=false
      drawAs=Box
      imageType=FullColor
      mirroring=NoMirror
      resourceName=../../../Engine/Content/Slate/Common/ScrollBoxShadowLeft.png
      resourceObject=null
      tiling=NoTile
      imageSize=<8 16>
      margin=<ltrb 1 0.5 0 0.5>
      outlineSettings {
       bUseBrushTransparency=false
       roundingType=HalfHeightRadius
       width=0
       color {colorUseRule=UseColor_Specified, specifiedColor=<rgba 0 0 0 0>}
       cornerRadii=<0 0 0 0>
      }
      tintColor {colorUseRule=UseColor_Specified, specifiedColor=<rgba 1 1 1 1>}
      uVRegion {bIsValid=false, max=<0 0>, min=<0 0>}
     }
     rightShadowBrush {
      bIsDynamicallyLoaded=false
      drawAs=Box
      imageType=FullColor
      mirroring=NoMirror
      resourceName=../../../Engine/Content/Slate/Common/ScrollBoxShadowRight.png
      resourceObject=null
      tiling=NoTile
      imageSize=<8 16>
      margin=<ltrb 0 0.5 1 0.5>
      outlineSettings {
       bUseBrushTransparency=false
       roundingType=HalfHeightRadius
       width=0
       color {colorUseRule=UseColor_Specified, specifiedColor=<rgba 0 0 0 0>}
       cornerRadii=<0 0 0 0>
      }
      tintColor {colorUseRule=UseColor_Specified, specifiedColor=<rgba 1 1 1 1>}
      uVRegion {bIsValid=false, max=<0 0>, min=<0 0>}
     }
     topShadowBrush {
      bIsDynamicallyLoaded=false
      drawAs=Box
      imageType=FullColor
      mirroring=NoMirror
      resourceName=../../../Engine/Content/Slate/Common/ScrollBoxShadowTop.png
      resourceObject=null
      tiling=NoTile
      imageSize=<16 8>
      margin=<ltrb 0.5 1 0.5 0>
      outlineSettings {
       bUseBrushTransparency=false
       roundingType=HalfHeightRadius
       width=0
       color {colorUseRule=UseColor_Specified, specifiedColor=<rgba 0 0 0 0>}
       cornerRadii=<0 0 0 0>
      }
      tintColor {colorUseRule=UseColor_Specified, specifiedColor=<rgba 1 1 1 1>}
      uVRegion {bIsValid=false, max=<0 0>, min=<0 0>}
     }
     verticalScrolledContentPadding=<ltrb 0 0 0 1>
    }
   }
   slot {class=VerticalBoxSlot, properties {padding=<ltrb 5 5 5 5>, size {sizeRule=Fill, value=1}}}
  }
  ActionDetailsView {widgetClassPath=&DetailsView, parent=ScrollBox_0, widget=ActionDetailsView, slot {class=ScrollBoxSlot}}
  Border_3 {
   widgetClassPath=&Border
   parent=ScrollBox_0
   widget=Border_3
   properties {
    background {
     bIsDynamicallyLoaded=false
     drawAs=Box
     imageType=NoImage
     mirroring=NoMirror
     resourceName=None
     resourceObject=null
     tiling=NoTile
     imageSize=<32 32>
     margin=<ltrb 0 0 0 0>
     outlineSettings {
      bUseBrushTransparency=false
      roundingType=HalfHeightRadius
      width=0
      color {colorUseRule=UseColor_Specified, specifiedColor=<rgba 0 0 0 0>}
      cornerRadii=<0 0 0 0>
     }
     tintColor {colorUseRule=UseColor_Specified, specifiedColor=<rgba 1 1 1 1>}
     uVRegion {bIsValid=false, max=<0 0>, min=<0 0>}
    }
    brushColor=<rgba 0 0 0 0.2>
    padding=<ltrb 5 5 5 5>
   }
   slot {class=ScrollBoxSlot}
  }
  HorizontalBox_3 {
   widgetClassPath=&HorizontalBox
   parent=Border_3
   widget=HorizontalBox_3
   slot {class=BorderSlot, properties {padding=<ltrb 5 5 5 5>}}
  }
  ConfirmButton {
   widgetClassPath=&Button
   parent=HorizontalBox_3
   widget=ConfirmButton
   properties {widgetStyle {
     clickedSlateSound {resourceObject=null}
     disabled {
      bIsDynamicallyLoaded=false
      drawAs=Box
      imageType=NoImage
      mirroring=NoMirror
      resourceName=None
      resourceObject=null
      tiling=NoTile
      imageSize=<0 0>
      margin=<ltrb 0 0 0 0>
      outlineSettings {
       bUseBrushTransparency=false
       roundingType=FixedRadius
       width=1
       color {colorUseRule=UseColor_ColorTable, specifiedColor=<rgba 0 0 0 0>}
       cornerRadii=<4 4 4 4>
      }
      tintColor {colorUseRule=UseColor_Specified, specifiedColor=<rgba 0.0395 0.0395 0.0395 1>}
      uVRegion {bIsValid=false, max=<0 0>, min=<0 0>}
     }
     disabledForeground {colorUseRule=UseColor_ColorTable, specifiedColor=<rgba 0 0 0 0>}
     hovered {
      bIsDynamicallyLoaded=false
      drawAs=Box
      imageType=NoImage
      mirroring=NoMirror
      resourceName=None
      resourceObject=null
      tiling=NoTile
      imageSize=<32 32>
      margin=<ltrb 0 0 0 0>
      outlineSettings {
       bUseBrushTransparency=false
       roundingType=FixedRadius
       width=1
       color {colorUseRule=UseColor_ColorTable, specifiedColor=<rgba 0 0 0 0>}
       cornerRadii=<4 4 4 4>
      }
      tintColor {colorUseRule=UseColor_Specified, specifiedColor=<rgba 0.0953 0.0953 0.0953 1>}
      uVRegion {bIsValid=false, max=<0 0>, min=<0 0>}
     }
     hoveredForeground {colorUseRule=UseColor_ColorTable, specifiedColor=<rgba 0 0 0 0>}
     hoveredSlateSound {resourceObject=null}
     normal {
      bIsDynamicallyLoaded=false
      drawAs=Box
      imageType=NoImage
      mirroring=NoMirror
      resourceName=None
      resourceObject=null
      tiling=NoTile
      imageSize=<32 32>
      margin=<ltrb 0 0 0 0>
      outlineSettings {
       bUseBrushTransparency=false
       roundingType=FixedRadius
       width=1
       color {colorUseRule=UseColor_ColorTable, specifiedColor=<rgba 0 0 0 0>}
       cornerRadii=<4 4 4 4>
      }
      tintColor {colorUseRule=UseColor_Specified, specifiedColor=<rgba 0.0284 0.0284 0.0284 1>}
      uVRegion {bIsValid=false, max=<0 0>, min=<0 0>}
     }
     normalForeground {colorUseRule=UseColor_ColorTable, specifiedColor=<rgba 0 0 0 0>}
     normalPadding=<ltrb 12 1.5 12 1.5>
     pressed {
      bIsDynamicallyLoaded=false
      drawAs=Box
      imageType=NoImage
      mirroring=NoMirror
      resourceName=None
      resourceObject=null
      tiling=NoTile
      imageSize=<32 32>
      margin=<ltrb 0 0 0 0>
      outlineSettings {
       bUseBrushTransparency=false
       roundingType=FixedRadius
       width=1
       color {colorUseRule=UseColor_ColorTable, specifiedColor=<rgba 0 0 0 0>}
       cornerRadii=<4 4 4 4>
      }
      tintColor {colorUseRule=UseColor_Specified, specifiedColor=<rgba 0.0284 0.0284 0.0284 1>}
      uVRegion {bIsValid=false, max=<0 0>, min=<0 0>}
     }
     pressedForeground {colorUseRule=UseColor_ColorTable, specifiedColor=<rgba 0 0 0 0>}
     pressedPadding=<ltrb 12 2.5 12 0.5>
     pressedSlateSound {resourceObject=null}
    }}
   slot {class=HorizontalBoxSlot, properties {padding=<ltrb 0.5 0 0.5 0>, size {sizeRule=Fill, value=1}}}
  }
  Text_1 {
   widgetClassPath=&UEFN_TextBlock
   bIsVariable=true
   parent=ConfirmButton
   widget=Text_1
   properties {
    clipping=ClipToBounds
    text=Confirm
    font {
     bForceMonospaced=false
     bMaterialIsStencil=false
     fontMaterial=null
     fontObject=&NotoSans
     letterSpacing=5
     monospacedWidth=1
     size=10
     skewAmount=0
     typefaceFontName=NotoSans-Regular
     outlineSettings {
      bApplyOutlineToDropShadows=false
      bMiteredCorners=false
      bSeparateFillAlpha=false
      outlineMaterial=null
      outlineSize=0
      outlineColor=<rgba 0 0 0 1>
     }
    }
   }
   slot {class=ButtonSlot, properties {padding=<ltrb 0 0 0 0>}}
  }
  CancelButton {
   widgetClassPath=&Button
   parent=HorizontalBox_3
   widget=CancelButton
   properties {widgetStyle {
     clickedSlateSound {resourceObject=null}
     disabled {
      bIsDynamicallyLoaded=false
      drawAs=Box
      imageType=NoImage
      mirroring=NoMirror
      resourceName=None
      resourceObject=null
      tiling=NoTile
      imageSize=<0 0>
      margin=<ltrb 0 0 0 0>
      outlineSettings {
       bUseBrushTransparency=false
       roundingType=FixedRadius
       width=1
       color {colorUseRule=UseColor_ColorTable, specifiedColor=<rgba 0 0 0 0>}
       cornerRadii=<4 4 4 4>
      }
      tintColor {colorUseRule=UseColor_Specified, specifiedColor=<rgba 0.0395 0.0395 0.0395 1>}
      uVRegion {bIsValid=false, max=<0 0>, min=<0 0>}
     }
     disabledForeground {colorUseRule=UseColor_ColorTable, specifiedColor=<rgba 0 0 0 0>}
     hovered {
      bIsDynamicallyLoaded=false
      drawAs=Box
      imageType=NoImage
      mirroring=NoMirror
      resourceName=None
      resourceObject=null
      tiling=NoTile
      imageSize=<32 32>
      margin=<ltrb 0 0 0 0>
      outlineSettings {
       bUseBrushTransparency=false
       roundingType=FixedRadius
       width=1
       color {colorUseRule=UseColor_ColorTable, specifiedColor=<rgba 0 0 0 0>}
       cornerRadii=<4 4 4 4>
      }
      tintColor {colorUseRule=UseColor_Specified, specifiedColor=<rgba 0.0953 0.0953 0.0953 1>}
      uVRegion {bIsValid=false, max=<0 0>, min=<0 0>}
     }
     hoveredForeground {colorUseRule=UseColor_ColorTable, specifiedColor=<rgba 0 0 0 0>}
     hoveredSlateSound {resourceObject=null}
     normal {
      bIsDynamicallyLoaded=false
      drawAs=Box
      imageType=NoImage
      mirroring=NoMirror
      resourceName=None
      resourceObject=null
      tiling=NoTile
      imageSize=<32 32>
      margin=<ltrb 0 0 0 0>
      outlineSettings {
       bUseBrushTransparency=false
       roundingType=FixedRadius
       width=1
       color {colorUseRule=UseColor_ColorTable, specifiedColor=<rgba 0 0 0 0>}
       cornerRadii=<4 4 4 4>
      }
      tintColor {colorUseRule=UseColor_Specified, specifiedColor=<rgba 0.0284 0.0284 0.0284 1>}
      uVRegion {bIsValid=false, max=<0 0>, min=<0 0>}
     }
     normalForeground {colorUseRule=UseColor_ColorTable, specifiedColor=<rgba 0 0 0 0>}
     normalPadding=<ltrb 12 1.5 12 1.5>
     pressed {
      bIsDynamicallyLoaded=false
      drawAs=Box
      imageType=NoImage
      mirroring=NoMirror
      resourceName=None
      resourceObject=null
      tiling=NoTile
      imageSize=<32 32>
      margin=<ltrb 0 0 0 0>
      outlineSettings {
       bUseBrushTransparency=false
       roundingType=FixedRadius
       width=1
       color {colorUseRule=UseColor_ColorTable, specifiedColor=<rgba 0 0 0 0>}
       cornerRadii=<4 4 4 4>
      }
      tintColor {colorUseRule=UseColor_Specified, specifiedColor=<rgba 0.0284 0.0284 0.0284 1>}
      uVRegion {bIsValid=false, max=<0 0>, min=<0 0>}
     }
     pressedForeground {colorUseRule=UseColor_ColorTable, specifiedColor=<rgba 0 0 0 0>}
     pressedPadding=<ltrb 12 2.5 12 0.5>
     pressedSlateSound {resourceObject=null}
    }}
   slot {class=HorizontalBoxSlot, properties {padding=<ltrb 0.5 0 0.5 0>, size {sizeRule=Fill, value=1}}}
  }
  Text_3 {
   widgetClassPath=&UEFN_TextBlock
   bIsVariable=true
   parent=CancelButton
   widget=Text_3
   properties {
    clipping=ClipToBounds
    text=Cancel
    font {
     bForceMonospaced=false
     bMaterialIsStencil=false
     fontMaterial=null
     fontObject=&NotoSans
     letterSpacing=5
     monospacedWidth=1
     size=10
     skewAmount=0
     typefaceFontName=NotoSans-Regular
     outlineSettings {
      bApplyOutlineToDropShadows=false
      bMiteredCorners=false
      bSeparateFillAlpha=false
      outlineMaterial=null
      outlineSize=0
      outlineColor=<rgba 0 0 0 1>
     }
    }
   }
   slot {class=ButtonSlot, properties {padding=<ltrb 0 0 0 0>}}
  }
 ]
}
