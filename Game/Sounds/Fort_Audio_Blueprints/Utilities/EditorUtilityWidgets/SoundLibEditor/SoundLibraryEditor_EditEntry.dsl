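; Edit-entry Blueprint for the Sound Library Editor utility tool, holding the sound, tags and library of the entries being edited.
; SetVariables takes a library entry, copies its sound, resolves the parent editor and its library through the entry's widget, and clears the tag-to-entry map.
; AddSoundLibraryEntryAndRefresh maps the entry by its gameplay tag, then adds every map key to the Tag container, which is never cleared here.
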(fn AddSoundLibraryEntryAndRefresh
  (params (SoundLibraryEntry :type Sound%20Library%20Editor%20Library%20Entry%20Object%20Reference))
  (bind _tagtolibraryentry (Variables|Default|GetTagToLibraryEntry))
  (Utilities|Map|Add _tagtolibraryentry (Variables|/Game/Sounds/Fort_Audio_Blueprints/Utilities/EditorUtilityWidgets/SoundLibEditor/SoundLibraryEditor_LibraryEntry.SoundLibraryEditor_LibraryEntry_C|GetTag SoundLibraryEntry) SoundLibraryEntry)
  (bind _keys (Utilities|Map|Keys _tagtolibraryentry))
  (for _array_element _keys
    (GameplayTags|AddGameplayTag (Variables|Default|GetTag) _array_element :!class /Script/GameplayTags.BlueprintGameplayTagLibrary)))

(fn SetVariables
  (params (Library :type Sound%20Library%20Editor%20Library%20Entry%20Object%20Reference))
  (bind _widget (Variables|/Game/Sounds/Fort_Audio_Blueprints/Utilities/EditorUtilityWidgets/SoundLibEditor/SoundLibraryEditor_LibraryEntry.SoundLibraryEditor_LibraryEntry_C|GetWidget Library))
  (bind _parent (Variables|/Game/Sounds/Fort_Audio_Blueprints/Utilities/EditorUtilityWidgets/SoundLibEditor/EUW_SoundLibraryEditor_LibraryEntry.EUW_SoundLibraryEditor_LibraryEntry_C|GetParent _widget))
  (Utilities|Map|Clear (Variables|Default|GetTagToLibraryEntry))
  (Variables|Default|SetSound (Variables|/Game/Sounds/Fort_Audio_Blueprints/Utilities/EditorUtilityWidgets/SoundLibEditor/SoundLibraryEditor_LibraryEntry.SoundLibraryEditor_LibraryEntry_C|GetSound Library))
  (Variables|Default|SetParentWidget _parent)
  (Variables|Default|SetLibrary (Variables|/Game/Sounds/Fort_Audio_Blueprints/Utilities/EditorUtilityWidgets/SoundLibEditor/EUW_SoundLibraryEditor.EUW_SoundLibraryEditor_C|GetLibrary _parent))
  (CallFunction|AddSoundLibraryEntryandRefresh Library))
