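; Editor utility widget for a single library entry in the Sound Library Editor, holding a tag, a sound and a transient flag.
; Its SetVariables function takes a gameplay tag, a soft sound reference and a boolean, and stores them in member variables.
; The gameplay tag passes through MakeSoundLibraryTag before storage, while the sound reference and flag are stored unchanged.
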(fn SetVariables
  (params (Tag :type Gameplay%20Tag%20Structure) (Sound :type Sound%20Base%20Soft%20Object%20Reference) (Transient :type Boolean))
  (Variables|Default|SetTag (SoundLibrary|MakeSoundLibraryTag Tag))
  (Variables|Default|SetSound Sound)
  (Variables|Default|SetTransient Transient))
